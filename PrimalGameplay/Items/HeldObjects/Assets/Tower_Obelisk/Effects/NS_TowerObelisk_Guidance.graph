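# Niagara guidance effect for the held Tower Obelisk: ribbon tendrils reach from the player toward the obelisk location.
# A looping emitter bursts a few particles per ribbon, sharing per-loop random seeds for lifetime, start offset and width.
# Hermite splines place particles by partition ID, extending the end with age toward a clamped obelisk position.
# Smoothed distance scales opacity, so the effect fades near the obelisk; the distance-based color ramp is disabled.
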NiagaraSystem /PrimalGameplay/Items/HeldObjects/Assets/Tower_Obelisk/Effects/NS_TowerObelisk_Guidance
dependencies {
 usedDataInterfaces [&NiagaraDataInterfaceCurve, &NiagaraDataInterfaceVectorCurve, &NiagaraDataInterfaceColorCurve]
 usedDynamicInputs [
  &self:SmoothInterpolateSystemVector
  &self:ClampTetherEndpoint
  &VectorLength
  &Subtract_Vector
 ]
 usedModules [
  &SystemState
  &self:SystemSpawnScript.SystemScriptSource.SystemScriptGraph.NiagaraNodeAssignment_0.SetVariables_#1
  &self:SystemSpawnScript.SystemScriptSource.SystemScriptGraph.NiagaraNodeAssignment_2.SetVariables_#2
  &self:SystemSpawnScript.SystemScriptSource.SystemScriptGraph.NiagaraNodeAssignment_1.SetVariables_#3
 ]
}
emitters {Grabbers {
  emitterSpawnScript [/SetVariables(
    Emitter.EmitterRandom:NiagaraInt32=/Niagara/DynamicInputs/Random/RandomInt(Randomness Mode:ENiagaraRandomnessMode=::NewEnumerator0, Random Seed:NiagaraInt32=0, Fixed Random Seed:NiagaraBool=false)
    Emitter.ParticlesPerRibbon=5
   ) {isSetParameters=true}]
  emitterUpdateScript [
   &EmitterState(Life Cycle Mode=::NewEnumerator1, Loop Duration=0.15, Recalculate Duration Each Loop=true, UseLoopDelay=false)
   &SpawnBurst_Instantaneous(Spawn Count=$Emitter.ParticlesPerRibbon, Use Spawn Probability=false)
  ]
  particleSpawnScript [
   /SetVariables(
    Particles.PositionAtSpawn:Vector3f=$User.PlayerLocation
    Particles.RandomSeedByRibbon:NiagaraInt32=&Add_Integer(A=$Emitter.LoopCount, B=$Emitter.EmitterRandom)
   ) {isSetParameters=true}
   &InitializeParticle(
    AdjustAlpha=true
    AdjustHue=true
    AdjustSaturation=true
    AdjustValue=true
    Alpha Scale Range:Vector2f=<0.4 0.75>
    Color=<rgba 0.44 0 1 0.3>
    Color Mode=::NewEnumerator3
    Hue Shift Range:Vector2f=<-0.35 0.35>
    Lifetime=&RandomRangeFloat(Minimum=1, Maximum=1.5, Randomness Mode=::NewEnumerator1, Random Seed=$Particles.RandomSeedByRibbon, Fixed Random Seed=true)
    Saturation Range:Vector2f=<0.3 0.5>
    UsePositionOffset=false
    Value Range:Vector2f=<1 2>
    Write Material Random=true
   )
   &self:PartitionSimple() {module=PartitionSimple}
   &DynamicMaterialParameters(
    Index 0 Param 1=&RandomRangeFloat(Minimum=0.5, Maximum=2)
    Index 0 Param 2=50
    Index 0 Param 3=50
    Module.Param0WriteEnabled=true
    Module.Param1WriteEnabled=true
    Module.Param2WriteEnabled=true
    Module.Param3WriteEnabled=false
   )
  ]
  particleUpdateScript [
   &ParticleState()
   /Game/Effects/Niagara/Modules/Curves/FN_HermiteCubicSpline_v2/FN_HermiteCubicSpline(
    Position Blend:NiagaraFloat=$Particles.Partition.NormalizedID
    Position End:Vector3f=/Niagara/DynamicInputs/Transforms/TransformPosition(
     Position=&Lerp_Vector(
      A=$System.SmoothPlayerLocation
      B=$System.ClampedObeliskPos
      Alpha=&Multiply_Float(
       A=$Particles.NormalizedAge
       B=&FloatFromCurve(
        FloatCurve=DataInterface() {propertyValues {Curve {keys [
            {interpMode=RCIM_Cubic, tangentMode=RCTM_User, time=0.0353, value=0.4815}
            {arriveTangent=0.4454, interpMode=RCIM_Cubic, leaveTangent=0.4454, tangentMode=RCTM_User, time=0.4663, value=0.5855}
            {arriveTangent=1.6504, interpMode=RCIM_Cubic, leaveTangent=1.6504, tangentMode=RCTM_User, time=1, value=1}
           ]}}}
        Scale Curve=&RandomRangeFloat(Minimum=0.75, Maximum=0.9)
       )
      )
     )
     Source Space=::World
     Destination Space=::Simulation
    )
    Position Start:Vector3f=&TransformPosition(
     Position=&Add_Vector(
      A=$User.PlayerLocation
      B=&RandomVector(Vector Scale=20, Randomness Mode=::NewEnumerator1, Random Seed=$Particles.RandomSeedByRibbon, Fixed Random Seed=true)
     )
     Source Space=::World
     Destination Space=::Simulation
    )
    Write Particles Position=true
   )
   &SolveForcesAndVelocity(Clamp Velocity=false, Limit Acceleration=false)
   /SetVariables(Particles.RibbonWidth=&FloatFromCurve(
     FloatCurve=DataInterface() {propertyValues {Curve {keys [
         {arriveTangent=-0.0373, interpMode=RCIM_Cubic, leaveTangent=-0.0373, tangentMode=RCTM_User, time=0.4281, value=1.0152}
         {arriveTangent=-3.2938, interpMode=RCIM_Cubic, leaveTangent=-3.2938, tangentMode=RCTM_User, time=1, value=0.2654}
        ]}}}
     CurveIndex=$Particles.Partition.NormalizedID
     Scale Curve=&FloatFromCurve(
      FloatCurve=DataInterface() {propertyValues {Curve {keys [
          {arriveTangent=3.3003, interpMode=RCIM_Cubic, leaveTangent=3.3003, tangentMode=RCTM_User, time=0, value=0.5}
          {arriveTangent=-0.0702, interpMode=RCIM_Cubic, leaveTangent=-0.0702, tangentMode=RCTM_User, time=0.4, value=1}
          {arriveTangent=-2.8182, interpMode=RCIM_Cubic, leaveTangent=-2.8182, tangentMode=RCTM_User, time=1, value=0}
         ]}}}
      Scale Curve=&RandomRangeFloat(Minimum=20, Maximum=45, Randomness Mode=::NewEnumerator1, Random Seed=$Particles.RandomSeedByRibbon, Fixed Random Seed=true)
     )
    )) {isSetParameters=true}
   &ScaleColor(
    Scale Alpha=&FloatFromCurve(
     FloatCurve=DataInterface() {propertyValues {Curve {keys [
         {arriveTangent=-0.0184, interpMode=RCIM_Cubic, leaveTangent=-0.0184, tangentMode=RCTM_User, time=0, value=0}
         {arriveTangent=-0.9399, interpMode=RCIM_Cubic, leaveTangent=-0.9399, tangentMode=RCTM_User, time=0.25, value=2}
         {arriveTangent=0.0229, interpMode=RCIM_Cubic, leaveTangent=0.0229, tangentMode=RCTM_User, time=1, value=0}
        ]}}}
     Scale Curve=&Multiply_Float(A=&ClampFloat(Float=&Multiply_Float(A=$System.SmoothDistance, B=0.0005)), B=1)
    )
    Scale RGB=&VectorFromCurve(VectorCurve=DataInterface() {propertyValues {
       XCurve {keys [{arriveTangent=-2.663, time=0, value=1}, <at 1 0>]}
       YCurve {keys [{arriveTangent=-3.5378, time=0, value=1}, <at 1 0>]}
       ZCurve {keys [{arriveTangent=-1.682, time=0, value=1}, <at 1 1.25>]}
      }})
    ScaleA=true
    ScaleRGB=true
   )
   &Color(Color=&ColorFromCurve(
     ColorCurve=DataInterface() {propertyValues {
       AlphaCurve {keys [<at 0 1 RCIM_Cubic>]}
       BlueCurve {keys [<at 0 0 RCIM_Cubic>,<at 0.4908 0>,<at 1 1 RCIM_Cubic>]}
       GreenCurve {keys [<at 0 0 RCIM_Cubic>,<at 0.4908 1>,<at 1 0 RCIM_Cubic>]}
       RedCurve {keys [<at 0 1 RCIM_Cubic>,<at 0.4908 0>,<at 1 0 RCIM_Cubic>]}
      }}
     CurveIndex=&ClampFloat(Float=&Multiply_Float(A=$System.SmoothDistance, B=0.0005))
    )) {enabled=false}
  ]
  rendererClasses [&NiagaraRibbonRendererProperties]
  renderers [{rendererClass=&NiagaraRibbonRendererProperties}]
 }}
summary {
 systemName=NS_TowerObelisk_Guidance
 emitters [{emitterName=Grabbers, rendererClasses [&NiagaraRibbonRendererProperties]}]
 userVariables [
  User.ObeliskLocation {type=Vector3f, description="", defaultValue {struct=Vector3f, value=<1500 20 75>}}
  User.PlayerLocation {type=Vector3f, description="", defaultValue {struct=Vector3f, value=<0 0 0>}}
 ]
}
