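# Niagara system for the Tower Obelisk loot drop effect, a short burst of glow, light and debris.
# Glow, light spike and corona sprite bursts form the flash, with a fainter corona lingering.
# GPU embers with curl noise and drag, plus spinning mesh fragments, fly out biased along +X; some fragments raise death events.
# Several emitters are disabled, including spline-driven Reachers ribbons and two Empty emitters, one following User.PlayerLocation.
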NiagaraSystem /PrimalGameplay/Items/HeldObjects/Assets/Tower_Obelisk/Effects/NS_TowerObelisk_LootDrop
dependencies {
 usedDataInterfaces [
  /Script/Niagara.NiagaraDataInterfaceSimpleCounter
  &NiagaraDataInterfaceCurve
  &NiagaraDataInterfaceColorCurve
  &NiagaraDataInterfaceVector2DCurve
  &NiagaraDataInterfaceVectorCurve
 ]
 usedDynamicInputs [
  &VectorLength
  &Multiply_Vector
  &NormalizeVector
 ]
 usedModules [
  &self:SystemSpawnScript.SystemScriptSource.SystemScriptGraph.NiagaraNodeAssignment_1.SetVariables_#1
  &SystemState
  &self:SystemSpawnScript.SystemScriptSource.SystemScriptGraph.NiagaraNodeAssignment_0.SetVariables_#2
  &self:SystemSpawnScript.SystemScriptSource.SystemScriptGraph.NiagaraNodeAssignment_2.SetVariables_#3
 ]
}
emitters {
 CoronaBurst {
  emitterUpdateScript [
   &EmitterState()
   &SpawnBurst_Instantaneous(Spawn Count=5, Use Spawn Probability=false)
  ]
  particleSpawnScript [
   &InitializeParticle(
    AdjustAlpha=true
    AdjustHue=true
    AdjustSaturation=true
    AdjustValue=false
    Alpha Scale Range:Vector2f=<0.5 1>
    Color=<rgba 1 0.63 0.63 1>
    Color Mode=::NewEnumerator3
    Hue Shift Range:Vector2f=<-1 1>
    Lifetime Max=0.4
    Lifetime Min=0.3
    Lifetime Mode=::NewEnumerator1
    Saturation Range:Vector2f=<0.75 1>
    Sprite Rotation Mode=::NewEnumerator2
    Sprite Size Mode=::NewEnumerator1
    Uniform Sprite Size Max=1000
    Uniform Sprite Size Min=600
    UsePositionOffset=false
    Write Material Random=true
   )
   /SetVariables(Particles.Counter:NiagaraInt32=/Niagara/DynamicInputs/DataInterface/GenerateIntFromCounter(SimpleCounter:NiagaraDataInterfaceSimpleCounter=DataInterface() {propertyValues {GpuSyncMode=None, InitialValue=0}})) {isSetParameters=true, module=SetVariables_#4}
  ]
  particleUpdateScript [
   &ParticleState()
   &ScaleSpriteSize(Scale Factor=&Vector2DFromFloat(Value=&FloatFromCurve(FloatCurve=DataInterface() {propertyValues {Curve {keys [{arriveTangent=-0.0533, interpMode=RCIM_Cubic, leaveTangent=-0.0533, tangentMode=RCTM_User, time=0.2, value=0}, <at 1 1 RCIM_Cubic>]}}})))
   &DynamicMaterialParameters(
    Index 0 Param 1=&RandomRangeFloat(Minimum=0.75, Maximum=1.5)
    Index 0 Param 2=&RandomRangeFloat(Minimum=-0.4, Maximum=0.4)
    Index 0 Param 3=1
    Index 0 Param 4=1
    Index 1 Param 1=&RandomRangeFloat(Minimum=3.2, Maximum=5)
    Index 1 Param 2=&FloatFromCurve(
     FloatCurve=DataInterface() {propertyValues {Curve {keys [
         {arriveTangent=10.7118, interpMode=RCIM_Cubic, leaveTangent=10.7118, tangentMode=RCTM_User, time=0, value=0}
         {interpMode=RCIM_Cubic, tangentMode=RCTM_User, time=0.2, value=1}
         <at 1 0>
        ]}}}
     Scale Curve=&RandomRangeFloat(Minimum=0.1, Maximum=0.7)
    )
    Index 1 Param 3=10
    Module.Param0WriteEnabled=true
    Module.Param0WriteEnabled1=true
    Module.Param1WriteEnabled=true
    Module.Param1WriteEnabled1=true
    Module.Param2WriteEnabled=true
    Module.Param2WriteEnabled1=true
    Module.Param3WriteEnabled=true
    Module.Param3WriteEnabled1=true
    Write Parameter Index 1=true
   )
   &ScaleColor(
    Scale Alpha=&FloatFromCurve(FloatCurve=DataInterface() {propertyValues {Curve {keys [
         <at 0 0 RCIM_Cubic>
         <at 0.1 1>
         {arriveTangent=-4.5009, interpMode=RCIM_Cubic, leaveTangent=-4.5009, tangentMode=RCTM_User, time=0.2, value=1}
         {arriveTangent=-1.1436, interpMode=RCIM_Cubic, leaveTangent=-1.1436, tangentMode=RCTM_User, time=0.3974, value=0.3057}
         {arriveTangent=-0.2386, interpMode=RCIM_Cubic, leaveTangent=-0.2386, tangentMode=RCTM_User, time=1.0011, value=0.002}
        ]}}})
    ScaleA=true
    ScaleRGB=false
   )
   &SpriteRotationRate(Rotation Rate=&Lerp_Float(A=&RandomRangeFloat(Minimum=-75, Maximum=75), B=0, Alpha=$Particles.NormalizedAge))
   &CameraOffset(Camera Offset Amount=20)
  ]
  rendererClasses [&NiagaraSpriteRendererProperties]
  renderers [{rendererClass=&NiagaraSpriteRendererProperties}]
 }
 CoronaLinger {
  emitterUpdateScript [
   &EmitterState()
   &SpawnBurst_Instantaneous(Spawn Count=5, Use Spawn Probability=false)
  ]
  particleSpawnScript [
   &InitializeParticle(
    AdjustAlpha=true
    AdjustHue=true
    AdjustSaturation=true
    AdjustValue=false
    Alpha Scale Range:Vector2f=<0.25 0.25>
    Color=<rgba 1 0.63 0.63 1>
    Color Mode=::NewEnumerator3
    Hue Shift Range:Vector2f=<-1 1>
    Lifetime=2
    Saturation Range:Vector2f=<0.75 1>
    Sprite Rotation Mode=::NewEnumerator2
    Sprite Size Mode=::NewEnumerator1
    Uniform Sprite Size Max=800
    Uniform Sprite Size Min=400
    UsePositionOffset=false
    Write Material Random=true
   )
   /SetVariables(Particles.Counter:NiagaraInt32=&GenerateIntFromCounter(SimpleCounter:NiagaraDataInterfaceSimpleCounter=DataInterface() {propertyValues {GpuSyncMode=None, InitialValue=0}})) {isSetParameters=true, module=SetVariables_#4}
  ]
  particleUpdateScript [
   &ParticleState()
   &ScaleSpriteSize(Scale Factor=&Vector2DFromFloat(Value=&FloatFromCurve(FloatCurve=DataInterface() {propertyValues {Curve {keys [
          {arriveTangent=-0.1134, interpMode=RCIM_Cubic, leaveTangent=-0.1134, tangentMode=RCTM_User, time=0.2, value=1}
          {arriveTangent=-0.6394, interpMode=RCIM_Cubic, leaveTangent=-0.6394, tangentMode=RCTM_User, time=1, value=0.75}
         ]}}})))
   &DynamicMaterialParameters(
    Index 0 Param 1=&RandomRangeFloat(Minimum=0.75, Maximum=1.5)
    Index 0 Param 2=&RandomRangeFloat(Minimum=-0.4, Maximum=0.4)
    Index 0 Param 3=1
    Index 0 Param 4=1
    Index 1 Param 1=&RandomRangeFloat(Minimum=3.2, Maximum=5)
    Index 1 Param 2=&FloatFromCurve(
     FloatCurve=DataInterface() {propertyValues {Curve {keys [
         {arriveTangent=10.7118, interpMode=RCIM_Cubic, leaveTangent=10.7118, tangentMode=RCTM_User, time=0, value=0}
         {interpMode=RCIM_Cubic, tangentMode=RCTM_User, time=0.2, value=1}
         <at 1 0>
        ]}}}
     Scale Curve=&RandomRangeFloat(Minimum=0.1, Maximum=0.7)
    )
    Index 1 Param 3=25
    Module.Param0WriteEnabled=true
    Module.Param0WriteEnabled1=true
    Module.Param1WriteEnabled=true
    Module.Param1WriteEnabled1=true
    Module.Param2WriteEnabled=true
    Module.Param2WriteEnabled1=true
    Module.Param3WriteEnabled=true
    Module.Param3WriteEnabled1=true
    Write Parameter Index 1=true
   )
   &ScaleColor(
    Scale Alpha=&FloatFromCurve(FloatCurve=DataInterface() {propertyValues {Curve {keys [
         <at 0 0 RCIM_Cubic>
         {interpMode=RCIM_Cubic, tangentMode=RCTM_User, time=0.0383, value=1}
         {arriveTangent=-0.5702, interpMode=RCIM_Cubic, leaveTangent=-0.5702, tangentMode=RCTM_User, time=0.1893, value=0.1865}
         {arriveTangent=-0.2386, interpMode=RCIM_Cubic, leaveTangent=-0.2386, tangentMode=RCTM_User, time=1.0011, value=0.002}
        ]}}})
    ScaleA=true
    ScaleRGB=false
   )
   &SpriteRotationRate(Rotation Rate=&Lerp_Float(A=&RandomRangeFloat(Minimum=-20, Maximum=20), B=0, Alpha=$Particles.NormalizedAge))
   &CameraOffset(Camera Offset Amount=20)
  ]
  rendererClasses [&NiagaraSpriteRendererProperties]
  renderers [{rendererClass=&NiagaraSpriteRendererProperties}]
 }
 Embers {
  simTarget=GPUComputeSim
  emitterUpdateScript [
   &EmitterState(Enable Distance Culling=true, MaxDistance=4500, Module.bUseMaxDistance=true, Module.bUseMinDistance=false, Scalability Mode=::NewEnumerator1)
   &SpawnBurst_Instantaneous(Spawn Count=150, Spawn Time=0.1, Use Spawn Probability=false)
  ]
  particleSpawnScript [
   /SetVariables(Transient.Scale:NiagaraFloat=&Power(a=&RandomRangeFloat(Minimum=0, Maximum=1), b=3)) {isSetParameters=true}
   &InitializeParticle(
    AdjustAlpha=true
    AdjustHue=true
    AdjustSaturation=true
    AdjustValue=true
    Alpha Scale Range:Vector2f=<0.1 1>
    Color=<rgba 0.44 0 1 5>
    Color Mode=::NewEnumerator3
    Hue Shift Range:Vector2f=<-0.35 0.35>
    Lifetime=&Lerp_Float(A=1, B=0.25, Alpha=$Transient.Scale)
    Saturation Range:Vector2f=<0.3 0.5>
    Sprite Rotation Mode=::NewEnumerator2
    Sprite Size Mode=::NewEnumerator3
    Uniform Sprite Size=&Lerp_Float(A=10, B=15, Alpha=$Transient.Scale)
    UsePositionOffset=false
    Value Range:Vector2f=<0.6 16>
    Write Material Random=true
   )
   /Niagara/Modules/Spawn/Initialization/InitializeParticle(
    Color=&RandomRangeLinearColor(Minimum=<rgba 0 0 1 0.1>, Maximum=<rgba 4 16 16 1>, Color Channel Mode=::NewEnumerator2)
    Lifetime=&RandomRangeFloat(Minimum=0.25, Maximum=1, Evaluation Type=::NewEnumerator1)
    Module.Write Color=true
    Module.Write Lifetime=true
    Module.Write Mass=true
    Module.Write Material Random=true
    Module.Write Position=false
    Module.Write Scale=false
    Module.Write SpriteRotation=true
    Module.Write SpriteSize=true
    Sprite Rotation=&RandomRangeFloat(Minimum=0, Maximum=360, Evaluation Type=::NewEnumerator1)
    Sprite Size=&Vector2DFromFloat(Value=&RandomRangeFloat(Minimum=2, Maximum=5, Evaluation Type=::NewEnumerator1))
   ) {enabled=false, module=InitializeParticle001}
   /Niagara/Modules/Spawn/Location/SphereLocation(
    MaskSpawnGroup=false
    Module.Override Randomness=false
    Sphere Radius=&RandomRangeFloat(Minimum=50, Maximum=50, Evaluation Type=::NewEnumerator1)
    Surface Only=false
   )
   &AddVelocity(Velocity=&Add_Vector(
     A=&Multiply_VectorByFloat(
      Vector=$Output.SphereLocation.SphereVector
      Float=&RandomRangeFloat(Minimum=-0.04, Maximum=10, Evaluation Type=::NewEnumerator1)
     )
     B=&Lerp_Vector(
      A=&RandomRangeVector(Minimum=<25 -250 0>, Maximum=<750 250 250>)
      B=&RandomRangeVector(Minimum=<100 -250 0>, Maximum=<1200 250 250>)
      Alpha=$Transient.Scale
     )
    ))
  ]
  particleUpdateScript [
   &ParticleState()
   &Drag(Drag=5, Use Linear Drag=true, Use Rotational Drag=false)
   &CurlNoiseForce(Noise Frequency=20, Noise Strength=2000, Pan Noise=false)
   &SolveForcesAndVelocity(Clamp Velocity=false, Limit Acceleration=false)
   &ScaleSpriteSize(Scale Factor=&Vector2DFromFloat(Value=&FloatFromCurve(FloatCurve=DataInterface() {propertyValues {Curve {keys [
          {arriveTangent=5.1621, interpMode=RCIM_Cubic, leaveTangent=5.1621, tangentMode=RCTM_User, time=-0.0044, value=0.2654}
          {arriveTangent=-0.2642, interpMode=RCIM_Cubic, leaveTangent=-0.2642, time=0.2257, value=1}
          <at 1 0>
         ]}}})))
   &SpriteRotationRate(Rotation Rate=&RandomRangeFloat(Minimum=-1, Maximum=1))
   &ScaleColor(Scale RGB=&Lerp_Vector(A=<2 1 1>, B=<0.5 0.5 2>, Alpha=$Particles.NormalizedAge), ScaleA=false, ScaleRGB=true)
   &ScaleSpriteSizeBySpeed(Max Scale Factor=<1 3>)
  ]
  rendererClasses [&NiagaraSpriteRendererProperties]
  renderers [{rendererClass=&NiagaraSpriteRendererProperties}]
 }
 Empty {
  bEnabled=false
  emitterUpdateScript [
   &EmitterState()
   &SpawnRate(SpawnRate=5, Use Spawn Probability=false) {enabled=false}
   &SpawnBurst_Instantaneous(
    Spawn Count=1
    Spawn Probability=&MakeCustomFloatFromBool(
     Bool=&SetBoolByIntComparison(A=$Emitter.LoopCount, B=0, Comparison Type=::NewEnumerator4)
     True Float=1
     False Float=0
    )
    Use Spawn Probability=true
   )
  ]
  particleSpawnScript [
   /Niagara/Modules/Spawn/Initialization/V2/InitializeParticle(Color Mode=::NewEnumerator0, Mesh Scale Mode=::NewEnumerator3, UsePositionOffset=false, Write Material Random=true)
   &SystemLocation(MaskSpawnGroup=false, Offset=$User.PlayerLocation, Offset Coordinate Space=::Simulation) {enabled=false}
   &Color(Color=&ColorFromCurve(
     ColorCurve=DataInterface() {propertyValues {
       AlphaCurve {keys [<at 0 1>]}
       BlueCurve {keys [<at 0 0>,<at 0.5 0.0261>,<at 1.0012 1>]}
       GreenCurve {keys [<at 0 0>,<at 0.5 1>,<at 1.0012 0>]}
       RedCurve {keys [<at 0 1>,<at 0.5 0>,<at 1.0012 0>]}
      }}
     CurveIndex=&Multiply_Float(A=$System.DistanceToPlayer, B=0.001)
    )) {enabled=false}
  ]
  particleUpdateScript [
   &ParticleState(Kill Particles When Lifetime Has Elapsed=false, Let Infinitely Lived Particles Die When Emitter Deactivates=true)
   /Niagara/Modules/Update/Orientation/V2/OrientMeshToVector(
    Look At Direction:Vector3f=&Subtract_Vector(A=Linked() {linkedVariable {name=Engine.Owner.Position, type=NiagaraPosition}}, B=$User.PlayerLocation)
    Override Quat=false
    UseOptionalSideAxis=false
    UseOptionalUpAxis=false
   ) {enabled=false}
   /SetVariables(Particles.Position=&Add_Vector(
     A=$User.PlayerLocation
     B=&RotateVectorByQuaternion(
      Vector To Rotate=&RandomRangeVector(Minimum=<0 -20 -20>, Maximum=<0 20 20>)
      Quaternion=$Particles.MeshOrientation
     )
    )) {enabled=false, isSetParameters=true}
   /SetVariables(Particles.Position=&self:SmoothInterpolateParticleVector(
     Initial Vector:Vector3f=$User.PlayerLocation
     Vector to Interpolate:Vector3f=&Add_Vector(A=$User.PlayerLocation, B=<0 -25 0>)
     Rate Of Change:NiagaraFloat=7
    )) {isSetParameters=true}
  ]
  rendererClasses [&NiagaraSpriteRendererProperties, &NiagaraMeshRendererProperties]
  renderers [
   {rendererClass=&NiagaraSpriteRendererProperties}
   {rendererClass=&NiagaraMeshRendererProperties, rendererIndex=1}
   {rendererClass=&NiagaraMeshRendererProperties, rendererIndex=2}
  ]
 }
 Empty001 {
  bEnabled=false
  emitterUpdateScript [
   &EmitterState()
   &SpawnRate(SpawnRate=2, Use Spawn Probability=false)
  ]
  particleSpawnScript [
   &InitializeParticle(Color Mode=::NewEnumerator0, Sprite Size Mode=::NewEnumerator3, Uniform Sprite Size=500, UsePositionOffset=false, Write Material Random=true)
   &SystemLocation(MaskSpawnGroup=false, Offset=<-100 0 0>)
  ]
  particleUpdateScript [&ParticleState()]
  rendererClasses [&NiagaraSpriteRendererProperties]
  renderers [{rendererClass=&NiagaraSpriteRendererProperties}]
 }
 FN_Sprite {
  bEnabled=false
  simTarget=GPUComputeSim
  emitterUpdateScript [
   &EmitterState()
   &SpawnBurst_Instantaneous(Spawn Count=20, Use Spawn Probability=false)
  ]
  particleSpawnScript [
   /Niagara/Modules/Spawn/Initialization/InitializeParticle(
    Color=&LerpLinearColors(
     StartColor=<rgba 0.0228 0 0.276 1>
     EndColor=&RandomRangeLinearColor(Minimum=<rgba 0 0.2865 0.2196 1>, Maximum=<rgba 0.1346 0 0.2448 1>)
     LerpFactor=&UniformAOrBFloat(A=0, B=&UniformAOrBFloat(A=0))
    )
    Lifetime=&UniformRangedFloat(Minimum=0.5, Maximum=1)
    Module.Write Color=true
    Module.Write Lifetime=true
    Module.Write Mass=false
    Module.Write Material Random=true
    Module.Write Position=false
    Module.Write Scale=false
    Module.Write SpriteRotation=true
    Module.Write SpriteSize=true
    Position=&SimulationPosition()
    Sprite Rotation=&UniformRangedFloat(Minimum=0, Maximum=360)
    Sprite Size=&Vector2DFromFloat(Value=&Lerp_Float(A=10, B=10, Alpha=&Power(a=&RandomRangeFloat(Minimum=0, Maximum=1), b=2)))
   )
   &SphereLocation(
    MaskSpawnGroup=false
    Module.Override Randomness=false
    Sphere Radius=&RandomRangeFloat(Minimum=50, Maximum=50, Evaluation Type=::NewEnumerator1)
    Surface Only=false
   )
   &AddVelocity(Velocity=&Add_Vector(
     A=&Multiply_VectorByFloat(
      Vector=$Output.SphereLocation.SphereVector
      Float=&RandomRangeFloat(Minimum=-0.04, Maximum=10, Evaluation Type=::NewEnumerator1)
     )
     B=&RandomRangeVector(Minimum=<25 -250 0>, Maximum=<750 250 250>)
    ))
  ]
  particleUpdateScript [
   &ParticleState()
   &AccelerationForce(Acceleration=&RandomRangeVector(Minimum=<0 0 -200>, Maximum=<0 0 -100>))
   &Drag(Drag=2, Use Linear Drag=true, Use Rotational Drag=false)
   &SolveForcesAndVelocity(Clamp Velocity=false, Limit Acceleration=false)
   &ScaleColor(
    Scale Alpha=&FloatFromCurve(FloatCurve=DataInterface() {propertyValues {Curve {keys [<at 0 0>,<at 0.25 1>]}}})
    Scale RGB=&VectorFromFloat(Value=&Lerp_Float(A=&RandomRangeFloat(Minimum=1000, Maximum=5), B=0, Alpha=$Particles.NormalizedAge))
    ScaleA=true
    ScaleRGB=true
   )
   &SpriteRotationRate(Rotation Rate=&RandomRangeFloat(Minimum=-2, Maximum=2))
   &ScaleSpriteSize(Scale Factor=&Vector2DFromFloat(Value=&FloatFromCurve(FloatCurve=DataInterface() {propertyValues {Curve {keys [<at 0 1 RCIM_Cubic>, {arriveTangent=-1.9892, interpMode=RCIM_Cubic, leaveTangent=-1.9892, tangentMode=RCTM_User, time=1, value=0}]}}})))
  ]
  rendererClasses [&NiagaraSpriteRendererProperties]
  renderers [{rendererClass=&NiagaraSpriteRendererProperties}]
 }
 Init_Glow {
  emitterUpdateScript [
   &EmitterState()
   &SpawnBurst_Instantaneous(Spawn Count=1, Use Spawn Probability=false)
  ]
  particleSpawnScript [
   &InitializeParticle(
    Color=<rgba 0 1.0256 4 0.15>
    Lifetime=0.15
    Module.Write Color=true
    Module.Write Lifetime=true
    Module.Write Mass=true
    Module.Write Material Random=true
    Module.Write Position=true
    Module.Write Scale=false
    Module.Write SpriteRotation=true
    Module.Write SpriteSize=true
    Sprite Rotation=&RandomRangeFloat(Minimum=0, Maximum=360)
    Sprite Size=&Vector2DFromFloat(Value=2000)
   )
   /SetVariables(Particles.Color2=<rgba 10.7378 1.9749 15 0.3>) {isSetParameters=true, module=SetVariables_#5}
   /SetVariables(Particles.SpriteSize2=&Vector2DFromFloat(Value=750)) {isSetParameters=true, module=SetVariables_#6}
   &DynamicMaterialParameters(
    Index 0 Param 1=0.7
    Index 0 Param 2=0.2
    Index 0 Param 3=1
    Index 0 Param 4=1
    Module.Param0WriteEnabled=true
    Module.Param1WriteEnabled=true
    Module.Param2WriteEnabled=true
    Module.Param3WriteEnabled=true
   )
  ]
  particleUpdateScript [
   &ParticleState()
   &ScaleColor(
    Scale Alpha=&FloatFromCurve(FloatCurve=DataInterface() {propertyValues {Curve {keys [
         {arriveTangent=-0.0285, interpMode=RCIM_Cubic, leaveTangent=-0.0285, tangentMode=RCTM_User, time=0.3415, value=1.0115}
         {arriveTangent=-0.9789, interpMode=RCIM_Cubic, leaveTangent=-0.9789, tangentMode=RCTM_User, time=0.4128, value=0.2782}
         <at 1 0>
        ]}}})
    ScaleA=true
    ScaleRGB=true
   )
   /SetVariables(Particles.Color2=&MultiplyLinearColor(
     Linear Color=$Particles.Initial.Color2
     Scale Factor=&MakeFloat4FromVectorAndFloat(
      "Vector (RGB)"=<1 1 1>
      "Float (Alpha)"=&FloatFromCurve(FloatCurve=DataInterface() {propertyValues {Curve {keys [{arriveTangent=-0.0285, interpMode=RCIM_Cubic, leaveTangent=-0.0285, tangentMode=RCTM_User, time=0.1018, value=1.0115}, <at 0.872 0>]}}})
     )
    )) {isSetParameters=true, module=SetVariables_#7}
   &ScaleColor(
    Color Value To Scale=$Particles.Color2
    Scale Alpha=&FloatFromCurve(FloatCurve=DataInterface() {propertyValues {Curve {keys [<at 0 0.5>,<at 1 0>]}}})
    ScaleA=true
    ScaleRGB=true
   ) {enabled=false, module=ScaleColor001}
   &ScaleSpriteSize(Scale Factor=&Vector2DFromFloat(Value=&FloatFromCurve(FloatCurve=DataInterface() {propertyValues {Curve {keys [
          {interpMode=RCIM_Cubic, leaveTangent=4.5689, tangentMode=RCTM_Break, time=0, value=0}
          {arriveTangent=-2.6364, interpMode=RCIM_Cubic, leaveTangent=-2.6364, tangentMode=RCTM_User, time=0.125, value=1.5}
          {arriveTangent=-0.2275, interpMode=RCIM_Cubic, leaveTangent=-0.2275, tangentMode=RCTM_User, time=1, value=0.9655}
         ]}}})))
   &CameraOffset(Camera Offset Amount=500)
   &SpriteRotationRate(Rotation Rate=&Lerp_Float(A=30, B=0, Alpha=$Particles.NormalizedAge))
  ]
  rendererClasses [&NiagaraSpriteRendererProperties]
  renderers [
   {rendererClass=&NiagaraSpriteRendererProperties}
   {rendererClass=&NiagaraSpriteRendererProperties, rendererIndex=1}
  ]
 }
 Init_Glow001 {
  bEnabled=false
  emitterUpdateScript [
   &EmitterState()
   &SpawnBurst_Instantaneous(Spawn Count=1, Use Spawn Probability=false)
  ]
  particleSpawnScript [
   &InitializeParticle(
    Color=<rgba 0 1.0256 4 1>
    Lifetime=0.4
    Module.Write Color=true
    Module.Write Lifetime=true
    Module.Write Mass=true
    Module.Write Material Random=true
    Module.Write Position=true
    Module.Write Scale=false
    Module.Write SpriteRotation=true
    Module.Write SpriteSize=true
    Sprite Rotation=&RandomRangeFloat(Minimum=0, Maximum=360)
    Sprite Size=&Vector2DFromFloat(Value=1500)
   )
   /SetVariables(Particles.Color2=<rgba 10.7378 1.9749 15 3>) {isSetParameters=true, module=SetVariables_#5}
   /SetVariables(Particles.SpriteSize2=&Vector2DFromFloat(Value=500)) {isSetParameters=true, module=SetVariables_#6}
   &DynamicMaterialParameters(
    Index 0 Param 1=0.7
    Index 0 Param 2=0.2
    Index 0 Param 3=1
    Index 0 Param 4=1
    Module.Param0WriteEnabled=true
    Module.Param1WriteEnabled=true
    Module.Param2WriteEnabled=true
    Module.Param3WriteEnabled=true
   )
  ]
  particleUpdateScript [
   &ParticleState()
   &ScaleColor(
    Scale Alpha=&FloatFromCurve(FloatCurve=DataInterface() {propertyValues {Curve {keys [
         {arriveTangent=-0.0285, interpMode=RCIM_Cubic, leaveTangent=-0.0285, tangentMode=RCTM_User, time=0.3415, value=1.0115}
         {arriveTangent=-0.9789, interpMode=RCIM_Cubic, leaveTangent=-0.9789, tangentMode=RCTM_User, time=0.4128, value=0.2782}
         <at 1 0>
        ]}}})
    ScaleA=true
    ScaleRGB=true
   )
   /SetVariables(Particles.Color2=&MultiplyLinearColor(
     Linear Color=$Particles.Initial.Color2
     Scale Factor=&MakeFloat4FromVectorAndFloat(
      "Vector (RGB)"=<1 1 1>
      "Float (Alpha)"=&FloatFromCurve(FloatCurve=DataInterface() {propertyValues {Curve {keys [{arriveTangent=-0.0285, interpMode=RCIM_Cubic, leaveTangent=-0.0285, tangentMode=RCTM_User, time=0.1018, value=1.0115}, <at 1 0>]}}})
     )
    )) {isSetParameters=true, module=SetVariables_#7}
   &ScaleColor(
    Color Value To Scale=$Particles.Color2
    Scale Alpha=&FloatFromCurve(FloatCurve=DataInterface() {propertyValues {Curve {keys [<at 0 0.5>,<at 1 0>]}}})
    ScaleA=true
    ScaleRGB=true
   ) {enabled=false, module=ScaleColor001}
   &ScaleSpriteSize(Scale Factor=&Vector2DFromFloat(Value=&FloatFromCurve(FloatCurve=DataInterface() {propertyValues {Curve {keys [
          {interpMode=RCIM_Cubic, leaveTangent=4.5689, tangentMode=RCTM_Break, time=0, value=0}
          {arriveTangent=0.2009, interpMode=RCIM_Cubic, leaveTangent=0.2009, tangentMode=RCTM_User, time=0.125, value=1}
          <at 0.3936 0.4982>
          <at 1 0.5>
         ]}}})))
   &CameraOffset(Camera Offset Amount=100)
   &SpriteRotationRate(Rotation Rate=&Lerp_Float(A=30, B=0, Alpha=$Particles.NormalizedAge))
  ]
  rendererClasses [&NiagaraSpriteRendererProperties]
  renderers [
   {rendererClass=&NiagaraSpriteRendererProperties}
   {rendererClass=&NiagaraSpriteRendererProperties, rendererIndex=1}
  ]
 }
 LightSpike {
  emitterUpdateScript [
   &EmitterState()
   &SpawnBurst_Instantaneous(Spawn Count=1, Spawn Time=0.03, Use Spawn Probability=false)
  ]
  particleSpawnScript [
   &InitializeParticle(
    Color=<rgba 0.1 3.8569 5 0.25>
    Lifetime=0.5
    Mass=&UniformRangedFloat(Minimum=0.6, Maximum=1.2)
    Module.Write Color=true
    Module.Write Lifetime=true
    Module.Write Mass=false
    Module.Write Material Random=true
    Module.Write Position=false
    Module.Write Scale=false
    Module.Write SpriteRotation=true
    Module.Write SpriteSize=true
    Position=&SimulationPosition()
    Sprite Rotation=-90
    Sprite Size=&Vector2DFromFloat(Value=2000)
   )
   &SystemLocation(MaskSpawnGroup=false, Offset=<-20 0 0>)
   /SetVariables(Particles.Size2:Vector2f=<350 8000>) {isSetParameters=true}
  ]
  particleUpdateScript [
   &ParticleState()
   &ScaleColor(
    Scale Alpha=&FloatFromCurve(FloatCurve=DataInterface() {propertyValues {Curve {keys [<at 0 0.5 RCIM_Cubic>,<at 1 0>]}}})
    Scale RGB=&MakeVectorFromLinearColorRGB(Color=&ColorFromCurve(ColorCurve=DataInterface() {propertyValues {
        AlphaCurve {keys [<at 0 1 RCIM_Cubic>]}
        BlueCurve {keys [{arriveTangent=0.3517, interpMode=RCIM_Cubic, time=0.0193, value=4}, <at 0.1705 1.75>, <at 0.4647 1>]}
        GreenCurve {keys [{arriveTangent=0.8075, interpMode=RCIM_Cubic, time=0.0193, value=3.2947}, <at 0.1705 1.1253>, <at 0.4647 0.4152>]}
        RedCurve {keys [{arriveTangent=3.7024, interpMode=RCIM_Cubic, time=0.0193, value=1.6492}, <at 0.1705 0.7088>, <at 0.4647 0.5029>]}
       }}))
    ScaleA=true
    ScaleRGB=true
   )
   &SolveForcesAndVelocity(Clamp Velocity=false, Limit Acceleration=false)
   &ScaleSpriteSize(Scale Factor=&Vector2DFromCurve(Vector2Curve=DataInterface() {propertyValues {XCurve {keys [<at 0 1>,<at 1 1>]}, YCurve {keys [<at 0 1>,<at 1 0>]}}}, ScaleCurve=<1 1>))
   &CameraOffset(Camera Offset Amount=50)
   &DynamicMaterialParameters(
    Index 0 Param 1=1
    Index 0 Param 3=0.35
    Index 1 Param 4=1000
    Module.Param0WriteEnabled=true
    Module.Param0WriteEnabled1=true
    Module.Param1WriteEnabled=true
    Module.Param1WriteEnabled1=true
    Module.Param2WriteEnabled=true
    Module.Param2WriteEnabled1=true
    Module.Param3WriteEnabled=true
    Module.Param3WriteEnabled1=true
    Write Parameter Index 1=true
   )
  ]
  rendererClasses [&NiagaraSpriteRendererProperties]
  renderers [
   {rendererClass=&NiagaraSpriteRendererProperties}
   {rendererClass=&NiagaraSpriteRendererProperties, rendererIndex=1}
  ]
 }
 Reachers {
  bEnabled=false
  emitterSpawnScript [/SetVariables(
    Emitter.EmitterRandom:NiagaraInt32=/Niagara/DynamicInputs/Random/RandomInt(Randomness Mode:ENiagaraRandomnessMode=::NewEnumerator0, Random Seed:NiagaraInt32=0, Fixed Random Seed:NiagaraBool=false)
    Emitter.ParticlesPerRibbon=4
   ) {isSetParameters=true}]
  emitterUpdateScript [
   &EmitterState(Life Cycle Mode=::NewEnumerator1, Loop Behavior=::NewEnumerator2, Loop Count=20, Loop Duration=0.0001, UseLoopDelay=false)
   &SpawnBurst_Instantaneous(Spawn Count=$Emitter.ParticlesPerRibbon, Use Spawn Probability=false)
  ]
  particleSpawnScript [
   /SetVariables(Particles.RandomSeedByRibbon:NiagaraInt32=&Add_Integer(A=$Emitter.LoopCount, B=$Emitter.EmitterRandom)) {isSetParameters=true}
   /SetVariables(Particles.EndOffset:Vector3f=&RotateVectorByQuaternion(
     Vector To Rotate=&Multiply_VectorByFloat(
      Vector=&RandomRangeVector(Minimum=<-0.1 -1 -1>, Maximum=<1 1 1>, Random Seed=$Particles.RandomSeedByRibbon, Fixed Random Seed=true)
      Float=20
     )
     Quaternion=$Engine.Owner.Rotation
    )) {isSetParameters=true}
   /Niagara/Modules/Spawn/Initialization/V2/InitializeParticle(
    AdjustAlpha=true
    AdjustHue=true
    AdjustSaturation=true
    AdjustValue=false
    Alpha Scale Range:Vector2f=&Vector2DFromCurve(
     Vector2Curve=DataInterface() {propertyValues {XCurve {keys [<at 0 1>,<at 0.1 0.05>]}, YCurve {keys [<at 0 1>,<at 0.1 0.2>]}}}
     CurveIndex=$Emitter.Age
     ScaleCurve=<1 1>
    )
    Color=<rgba 0.44 0 1 0.5>
    Color Mode=::NewEnumerator3
    Hue Shift Range:Vector2f=<-0.35 0.35>
    Lifetime=&RandomRangeFloat(Minimum=0.75, Maximum=1.25, Randomness Mode=::NewEnumerator1, Random Seed=$Particles.RandomSeedByRibbon, Fixed Random Seed=true)
    Saturation Range:Vector2f=<0.3 0.5>
    Sprite Size Mode=::NewEnumerator3
    Uniform Sprite Size=500
    UsePositionOffset=false
    Write Material Random=true
   )
   &self:PartitionSimple() {module=PartitionSimple}
  ]
  particleUpdateScript [
   &ParticleState()
   /Game/Effects/Niagara/Modules/Curves/FN_HermiteCubicSpline_v2/FN_HermiteCubicSpline(
    Direction End:Vector3f=&RotateVectorByQuaternion(
     Vector To Rotate=&Multiply_VectorByFloat(
      Vector=&MakeVector(
       X=&Sine(
        Normalized Angle=$Particles.Age
        Period=5
        Scale=&RandomRangeFloat(Minimum=-31.25, Maximum=31.25)
        Bias=0
       )
       Y=&Cosine(Normalized Angle=$Particles.Age, Period=4, Scale=&RandomRangeFloat(Minimum=-75, Maximum=75), Bias=0)
       Z=&Sine(Normalized Angle=$Particles.Age, Period=6, Scale=&RandomRangeFloat(Minimum=-75, Maximum=75), Bias=0)
      )
      Float=$Particles.NormalizedAge
     )
     Quaternion=$Engine.Owner.Rotation
    )
    Direction Start:Vector3f=&RandomVector(Vector Scale=50)
    Position Blend:NiagaraFloat=$Particles.Partition.NormalizedID
    Position End:Vector3f=&Multiply_VectorByFloat(Vector=$Particles.EndOffset, Float=&Lerp_Float(A=0, B=100, Alpha=$Particles.NormalizedAge))
    Position Start:Vector3f=<0 0 0>
    Write Particles Position=true
   ) {enabled=false, module=FN_HermiteCubicSpline001}
   &FN_HermiteCubicSpline(
    Direction Start:Vector3f=&RandomVector(Vector Scale=50)
    Position Blend:NiagaraFloat=$Particles.Partition.NormalizedID
    Position End:Vector3f=&Multiply_VectorByFloat(
     Vector=$Particles.EndOffset
     Float=&FloatFromCurve(
      FloatCurve=DataInterface() {propertyValues {Curve {keys [
          {arriveTangent=25.4601, interpMode=RCIM_Cubic, leaveTangent=25.4601, tangentMode=RCTM_User, time=0, value=0}
          {arriveTangent=0.4527, interpMode=RCIM_Cubic, leaveTangent=0.4527, tangentMode=RCTM_User, time=0.0516, value=0.8557}
          {arriveTangent=0.0018, interpMode=RCIM_Cubic, leaveTangent=0.0018, tangentMode=RCTM_User, time=1, value=1}
         ]}}}
      Scale Curve=7
     )
    )
    Position Start:Vector3f=&Multiply_VectorByFloat(
     Vector=$Particles.EndOffset
     Float=&FloatFromCurve(FloatCurve=DataInterface() {propertyValues {Curve {keys [
          {arriveTangent=2.8705, interpMode=RCIM_Cubic, leaveTangent=2.8705, tangentMode=RCTM_User, time=0, value=0}
          {arriveTangent=1.0434, interpMode=RCIM_Cubic, leaveTangent=1.0434, tangentMode=RCTM_User, time=0.348, value=0.6357}
          <at 1 1>
         ]}}})
    )
    Write Particles Position=true
   )
   &SolveForcesAndVelocity(Clamp Velocity=false, Limit Acceleration=false)
   /SetVariables(Particles.RibbonWidth=&FloatFromCurve(
     FloatCurve=DataInterface() {propertyValues {Curve {keys [
         {arriveTangent=1.7254, interpMode=RCIM_Cubic, leaveTangent=1.7254, tangentMode=RCTM_User, time=0, value=0.75}
         {arriveTangent=-0.0448, interpMode=RCIM_Cubic, leaveTangent=-0.0448, tangentMode=RCTM_User, time=0.4, value=1}
         {arriveTangent=-0.3045, interpMode=RCIM_Cubic, leaveTangent=-0.3045, tangentMode=RCTM_User, time=1, value=0.5}
        ]}}}
     CurveIndex=$Particles.Partition.NormalizedID
     Scale Curve=&FloatFromCurve(
      FloatCurve=DataInterface() {propertyValues {Curve {keys [
          {arriveTangent=2.6006, interpMode=RCIM_Cubic, leaveTangent=2.6006, tangentMode=RCTM_User, time=0, value=0}
          {arriveTangent=0.6048, interpMode=RCIM_Cubic, leaveTangent=0.6048, tangentMode=RCTM_User, time=0.4, value=0.7182}
          {arriveTangent=0.3328, interpMode=RCIM_Cubic, leaveTangent=0.3328, tangentMode=RCTM_User, time=1, value=1}
         ]}}}
      Scale Curve=&RandomRangeFloat(Minimum=20, Maximum=80, Randomness Mode=::NewEnumerator1, Random Seed=$Particles.RandomSeedByRibbon, Fixed Random Seed=true)
     )
    )) {isSetParameters=true}
   &ScaleColor(
    Scale Alpha=&FloatFromCurve(
     FloatCurve=DataInterface() {propertyValues {Curve {keys [
         {arriveTangent=4.6804, interpMode=RCIM_Cubic, leaveTangent=4.6804, tangentMode=RCTM_User, time=-0.0001, value=0}
         {arriveTangent=-0.0174, interpMode=RCIM_Cubic, leaveTangent=-0.0174, tangentMode=RCTM_User, time=0.2, value=1}
         {arriveTangent=-3.0907, interpMode=RCIM_Cubic, leaveTangent=-3.0907, tangentMode=RCTM_User, time=1, value=0}
        ]}}}
     Scale Curve=&FloatFromCurve(
      FloatCurve=DataInterface() {propertyValues {Curve {keys [
          {arriveTangent=1.4642, interpMode=RCIM_Cubic, leaveTangent=1.4642, tangentMode=RCTM_User, time=-0.0001, value=0}
          {arriveTangent=-2.1333, interpMode=RCIM_Cubic, leaveTangent=-2.1333, tangentMode=RCTM_User, time=0.3, value=1}
          <at 1 0>
         ]}}}
      CurveIndex=$Particles.Partition.NormalizedID
     )
    )
    Scale RGB=&VectorFromCurve(VectorCurve=DataInterface() {propertyValues {
       XCurve {keys [
         <at 0 4 RCIM_Cubic>
         {arriveTangent=-1.7022, interpMode=RCIM_Cubic, leaveTangent=-1.7022, tangentMode=RCTM_User, time=0.1946, value=0.8054}
         <at 1 0.5>
        ]}
       YCurve {keys [
         <at 0 4 RCIM_Cubic>
         {arriveTangent=-1.7596, interpMode=RCIM_Cubic, leaveTangent=-1.7596, tangentMode=RCTM_User, time=0.1946, value=0.8054}
         <at 1 0.5>
        ]}
       ZCurve {keys [
         <at 0 4 RCIM_Cubic>
         {arriveTangent=-0.9104, interpMode=RCIM_Cubic, leaveTangent=-0.9104, tangentMode=RCTM_User, time=0.1946, value=1.0486}
         <at 1 1.25>
        ]}
      }})
    ScaleA=true
    ScaleRGB=true
   )
   &Color(Color=&ColorFromCurve(ColorCurve=DataInterface() {propertyValues {
       AlphaCurve {keys [<at 0 1 RCIM_Cubic>]}
       BlueCurve {keys [<at 0 0 RCIM_Cubic>,<at 0.4908 0>,<at 1 1 RCIM_Cubic>]}
       GreenCurve {keys [<at 0 0 RCIM_Cubic>,<at 0.4908 1>,<at 1 0 RCIM_Cubic>]}
       RedCurve {keys [<at 0 1 RCIM_Cubic>,<at 0.4908 0>,<at 1 0 RCIM_Cubic>]}
      }})) {enabled=false}
  ]
  rendererClasses [&NiagaraRibbonRendererProperties]
  renderers [{rendererClass=&NiagaraRibbonRendererProperties}]
 }
 TinyFrags {
  emitterUpdateScript [
   &EmitterState()
   &SpawnBurst_Instantaneous(Spawn Count=30, Spawn Time=0.1, Use Spawn Probability=false)
  ]
  particleSpawnScript [
   /Niagara/Modules/Spawn/Initialization/InitializeParticle(
    Color=&RandomRangeLinearColor(Minimum=<rgba 5 0 3 1>, Maximum=<rgba 0 6 6 1>)
    Lifetime=&RandomRangeFloat(Minimum=0.15, Maximum=1.25)
    Mesh Scale=&VectorFromFloat(Value=&Lerp_Float(A=0.75, B=4, Alpha=&Power(a=&RandomRangeFloat(Minimum=0, Maximum=1), b=3)))
    Module.Write Color=true
    Module.Write Lifetime=true
    Module.Write Mass=true
    Module.Write Material Random=true
    Module.Write Position=false
    Module.Write Scale=true
    Module.Write SpriteRotation=false
    Module.Write SpriteSize=false
   )
   /SetVariables(Particles.MeshOrientation=&MakeQuatFromAxisAngle(
     Axis=&RandomRangeVector(Minimum=<0 0 0>, Maximum=<1 1 1>, Evaluation Type=::NewEnumerator1)
     AngleInDegrees=90
    )) {isSetParameters=true, module=SetVariables_#8}
   /SetVariables(Particles.SpawnRotationVector=&RandomRangeVector(Minimum=<-0.15 -0.15 -0.15>, Maximum=<0.15 0.15 0.15>, Evaluation Type=::NewEnumerator1)) {isSetParameters=true, module=SetVariables_#9}
   &Color(Color=&MakeLinearColorFromVectorAndFloat("Vector (RGB)"=<9.3572 4.2143 10>, "Float (Alpha)"=&ClampFloat(Float=1))) {enabled=false}
   &SphereLocation(
    Hemisphere X=true
    MaskSpawnGroup=false
    Module.Override Randomness=false
    Sphere Radius=&RandomRangeFloat(Minimum=10, Maximum=60, Evaluation Type=::NewEnumerator1)
    Surface Only=true
   )
   &AddVelocityFromPoint(
    Module.Use Falloff Distance=false
    Origin Offset=<-40 0 -35>
    Velocity Strength=&Lerp_Float(
     A=50
     B=1000
     Alpha=&Power(a=&RandomRangeFloat(Minimum=0, Maximum=1), b=0.5)
    )
   )
   /SetVariables(Particles.GenerateDeathEvent:NiagaraBool=&Boolean_AndOperation(
     Bool1=&RandomBool()
     Bool2=&Boolean_OrOperation(Bool1=&RandomBool(), Bool2=&RandomBool())
    )) {enabled=false, isSetParameters=true}
   /SetVariables(Particles.GenerateDeathEvent:NiagaraBool=&Boolean_AndOperation(
     Bool1=&RandomBool()
     Bool2=&Boolean_OrOperation(
      Bool1=&RandomBool()
      Bool2=&Boolean_OrOperation(
       Bool1=&RandomBool()
       Bool2=&SetBoolByFloatComparison(A=$Particles.Lifetime, B=0.5)
      )
     )
    )) {enabled=false, isSetParameters=true}
   /SetVariables(Particles.GenerateDeathEvent:NiagaraBool=&RandomBool()) {isSetParameters=true}
  ]
  particleUpdateScript [
   &ParticleState()
   &GravityForce(Gravity=<0 0 -600>) {enabled=false}
   &Drag(Drag=&Lerp_Float(A=2, B=3, Alpha=$Emitter.Age), Use Linear Drag=true, Use Rotational Drag=false)
   &SolveForcesAndVelocity(Clamp Velocity=false, Limit Acceleration=false)
   &ScaleColor(
    Scale Alpha=&FloatFromCurve(FloatCurve=DataInterface() {propertyValues {Curve {keys [
         <at 0 0 RCIM_Cubic>
         {arriveTangent=1.4286, interpMode=RCIM_Cubic, leaveTangent=1.4286, time=0.2, value=1}
         {arriveTangent=-1.25, interpMode=RCIM_Cubic, leaveTangent=-1.25, time=0.7, value=1}
         <at 1 0>
        ]}}})
    ScaleA=true
    ScaleRGB=false
   )
   &UpdateMeshOrientation(
    Rotation Rate=&FloatFromCurve(FloatCurve=DataInterface() {propertyValues {Curve {keys [<at 0 1>,<at 0.75 0.1>]}}}, CurveIndex=$Emitter.Age, Scale Curve=10)
    Rotation Vector=$Particles.SpawnRotationVector
   ) {module=UpdateMeshOrientation001}
   &ScaleMeshSize(Scale Factor=&VectorFromFloat(Value=&MakeCustomFloatFromBool(
      Bool=$Particles.GenerateDeathEvent
      True Float=&FloatFromCurve(FloatCurve=DataInterface() {propertyValues {Curve {keys [
           <at 0 0 RCIM_Cubic>
           {arriveTangent=3.6185, interpMode=RCIM_Cubic, leaveTangent=3.6185, tangentMode=RCTM_User, time=0.2, value=1}
           {leaveTangent=-1.25, time=0.3055, value=1}
          ]}}})
      False Float=&FloatFromCurve(FloatCurve=DataInterface() {propertyValues {Curve {keys [
           <at 0 0 RCIM_Cubic>
           {arriveTangent=4.2555, interpMode=RCIM_Cubic, leaveTangent=4.2555, tangentMode=RCTM_User, time=0.2, value=1}
           {arriveTangent=-1.2187, interpMode=RCIM_Cubic, leaveTangent=-1.2187, tangentMode=RCTM_User, time=0.3055, value=1}
           {arriveTangent=3.2641, interpMode=RCIM_Cubic, leaveTangent=3.2641, tangentMode=RCTM_User, time=0.645, value=1}
           {arriveTangent=-3.3635, interpMode=RCIM_Cubic, leaveTangent=-3.3635, tangentMode=RCTM_User, time=0.815, value=1}
           <at 1 0>
          ]}}})
     ))) {enabled=false}
   &GenerateDeathEvent(Event Condition:NiagaraBool=$Particles.GenerateDeathEvent)
   &GenerateDeathEvent(Event Condition:NiagaraBool=&Boolean_AndOperation(
     Bool1=&SetBoolByFloatComparison(A=$Particles.Lifetime, B=0.3)
     Bool2=&SetBoolByFloatComparison(A=$Particles.Lifetime, B=0.75, Comparison Type=::NewEnumerator6)
    )) {enabled=false, module=GenerateDeathEvent001}
   &ScaleMeshSize(Scale Factor=&VectorFromFloat(Value=&FloatFromCurve(FloatCurve=DataInterface() {propertyValues {Curve {keys [<at 0 1>,<at 1 0>]}}}))) {module=ScaleMeshSize001}
  ]
  rendererClasses [&NiagaraMeshRendererProperties]
  renderers [{rendererClass=&NiagaraMeshRendererProperties, rendererIndex=0}]
 }
 TinyFrags002 {
  bEnabled=false
  emitterUpdateScript [
   &EmitterState()
   &SpawnBurst_Instantaneous(Spawn Count=30, Use Spawn Probability=false)
  ]
  particleSpawnScript [
   /SetVariables(Particles.MeshOrientation=&MakeQuatFromAxisAngle(
     Axis=&RandomRangeVector(Minimum=<0 0 0>, Maximum=<1 1 1>, Evaluation Type=::NewEnumerator1)
     AngleInDegrees=90
    )) {isSetParameters=true, module=SetVariables_#8}
   /SetVariables(Particles.SpawnRotationVector=&RandomRangeVector(Minimum=<-0.15 -0.15 -0.15>, Maximum=<0.15 0.15 0.15>, Evaluation Type=::NewEnumerator1)) {isSetParameters=true, module=SetVariables_#9}
   &InitializeParticle(
    Color=&RandomRangeLinearColor(Minimum=<rgba 0 0 0 1>, Maximum=<rgba 6 6 6 1>)
    Lifetime=&MakeCustomFloatFromBool(
     Bool=&SetBoolByFloatComparison(A=&ReturnNormalizedExecIndex(), B=0.2, Comparison Type=::NewEnumerator6)
     True Float=&RandomRangeFloat(Minimum=0.5, Maximum=1, Evaluation Type=::NewEnumerator1)
     False Float=&RandomRangeFloat(Minimum=0.2, Maximum=0.5, Evaluation Type=::NewEnumerator1)
    )
    Mesh Scale=&VectorFromFloat(Value=&Lerp_Float(A=0.75, B=4, Alpha=&Power(a=&RandomRangeFloat(Minimum=0, Maximum=1), b=3)))
    Module.Write Color=true
    Module.Write Lifetime=true
    Module.Write Mass=true
    Module.Write Material Random=true
    Module.Write Position=false
    Module.Write Scale=true
    Module.Write SpriteRotation=false
    Module.Write SpriteSize=false
   )
   &Color(Color=&MakeLinearColorFromVectorAndFloat("Vector (RGB)"=<9.3572 4.2143 10>, "Float (Alpha)"=&ClampFloat(Float=1))) {enabled=false}
   &SphereLocation(
    Hemisphere X=true
    MaskSpawnGroup=false
    Module.Override Randomness=false
    Sphere Radius=&RandomRangeFloat(Minimum=10, Maximum=60, Evaluation Type=::NewEnumerator1)
    Surface Only=true
   )
   &AddVelocityFromPoint(Module.Use Falloff Distance=false, Velocity Strength=&RandomRangeFloat(Minimum=50, Maximum=1000))
   &AddVelocityInCone(
    Cone Angle=10
    Module.Use Velocity Falloff On Cone Axis=false
    Velocity Strength=&UniformAOrBFloat(A=&RandomRangeFloat(Minimum=100, Maximum=1000), B=0)
   )
  ]
  particleUpdateScript [
   &ParticleState()
   &Drag(Drag=&Lerp_Float(A=3, B=10, Alpha=$Emitter.Age), Use Linear Drag=true, Use Rotational Drag=false)
   &SolveForcesAndVelocity(Clamp Velocity=false, Limit Acceleration=false)
   &ScaleColor(
    Scale Alpha=&FloatFromCurve(FloatCurve=DataInterface() {propertyValues {Curve {keys [
         <at 0 0 RCIM_Cubic>
         {arriveTangent=1.4286, interpMode=RCIM_Cubic, leaveTangent=1.4286, time=0.2, value=1}
         {arriveTangent=-1.25, interpMode=RCIM_Cubic, leaveTangent=-1.25, time=0.7, value=1}
         <at 1 0>
        ]}}})
    ScaleA=true
    ScaleRGB=false
   )
   &UpdateMeshOrientation(
    Rotation Rate=&FloatFromCurve(FloatCurve=DataInterface() {propertyValues {Curve {keys [<at 0 1>,<at 0.75 0.1>]}}}, CurveIndex=$Emitter.Age, Scale Curve=10)
    Rotation Vector=$Particles.SpawnRotationVector
   ) {module=UpdateMeshOrientation001}
   &ScaleMeshSize(Scale Factor=&VectorFromFloat(Value=&FloatFromCurve(
      FloatCurve=DataInterface() {propertyValues {Curve {keys [<at 0 0 RCIM_Cubic>, <at 0.2 1>, {arriveTangent=-1.25, interpMode=RCIM_Cubic, leaveTangent=-1.25, time=0.8, value=1}, <at 1 0>]}}}
      Scale Curve=&FloatFromCurve(
       FloatCurve=DataInterface() {propertyValues {Curve {keys [{interpMode=RCIM_Cubic, tangentMode=RCTM_Break, time=0, value=1}, <at 60000 8>, <at 70000 0>]}}}
       CurveIndex=$Engine.Owner.LODDistance
      )
     )))
   &GenerateDeathEvent(Event Condition:NiagaraBool=&Boolean_AndOperation(
     Bool1=&SetBoolByFloatComparison(A=$Particles.Lifetime, B=0.1)
     Bool2=&SetBoolByFloatComparison(A=$Particles.Lifetime, B=0.75, Comparison Type=::NewEnumerator6)
    ))
  ]
  rendererClasses [&NiagaraMeshRendererProperties]
  renderers [{rendererClass=&NiagaraMeshRendererProperties, rendererIndex=0}]
 }
 TinyFragsDeath {
  emitterUpdateScript [&EmitterState()]
  particleSpawnScript [
   /SetVariables(Particles.MeshOrientation=&MakeQuatFromAxisAngle(
     Axis=&RandomRangeVector(Minimum=<0 0 0>, Maximum=<1 1 1>, Evaluation Type=::NewEnumerator1)
     AngleInDegrees=90
    )) {enabled=false, isSetParameters=true, module=SetVariables_#8}
   /SetVariables(Particles.SpawnRotationVector=&RandomRangeVector(Minimum=<-0.15 -0.15 -0.15>, Maximum=<0.15 0.15 0.15>, Evaluation Type=::NewEnumerator1)) {enabled=false, isSetParameters=true, module=SetVariables_#9}
   /Niagara/Modules/Spawn/Initialization/V2/InitializeParticle(
    AdjustAlpha=true
    AdjustHue=true
    AdjustSaturation=true
    AdjustValue=true
    Alpha Scale Range:Vector2f=<0.25 1>
    Color=<rgba 0.44 0 1 1>
    Color Mode=::NewEnumerator3
    Hue Shift Range:Vector2f=<-0.35 0.35>
    Lifetime Max=1.5
    Lifetime Min=0.8
    Lifetime Mode=::NewEnumerator1
    Saturation Range:Vector2f=<0.3 1>
    Sprite Rotation Mode=::NewEnumerator2
    Sprite Size Mode=::NewEnumerator3
    Uniform Sprite Size=&RandomRangeFloat(Minimum=3, Maximum=7)
    UsePositionOffset=false
    Value Range:Vector2f=<0.8 5>
    Write Material Random=true
   )
   /SetVariables(Particles.BurstColor:LinearColor=&RandomRangeLinearColor(Minimum=<rgba 0.27 1 0.9635 1>, Maximum=<rgba 0.5 0.557 1 1>, Color Channel Mode=::NewEnumerator2)) {isSetParameters=true}
   /Niagara/Modules/Spawn/Initialization/InitializeParticle(
    Color=&RandomRangeLinearColor(Minimum=<rgba 0 0 0 1>, Maximum=<rgba 6 6 6 1>)
    Lifetime=&RandomRangeFloat(Minimum=0.1, Maximum=0.25, Evaluation Type=::NewEnumerator1)
    Mesh Scale=&VectorFromFloat(Value=&Lerp_Float(
      A=0.75
      B=1.5
      Alpha=&Power(a=&RandomRangeFloat(Minimum=0, Maximum=1), b=3)
     ))
    Module.Write Color=true
    Module.Write Lifetime=true
    Module.Write Mass=true
    Module.Write Material Random=true
    Module.Write Position=true
    Module.Write Scale=true
    Module.Write SpriteRotation=false
    Module.Write SpriteSize=true
    Sprite Size=&Vector2DFromFloat(Value=&RandomRangeFloat(Minimum=2, Maximum=5))
   ) {enabled=false, module=InitializeParticle001}
   &AddVelocityFromPoint(
    Module.Use Falloff Distance=false
    Origin Offset Coordinate Space=::World
    Velocity Strength=&RandomRangeFloat(Minimum=100, Maximum=1000)
   )
  ]
  particleUpdateScript [
   &ParticleState()
   &Drag(Drag=4, Use Linear Drag=true, Use Rotational Drag=false)
   &GravityForce(Gravity=&Lerp_Vector(
     A=<0 0 -980>
     B=&RandomRangeVector(Minimum=<0 0 500>, Maximum=<1 1 900>)
     Alpha=$Particles.NormalizedAge
    ))
   &SolveForcesAndVelocity(Clamp Velocity=false, Limit Acceleration=false)
   &ScaleColor(
    Scale Alpha=&FloatFromCurve(FloatCurve=DataInterface() {propertyValues {Curve {keys [
         {arriveTangent=1.4286, interpMode=RCIM_Cubic, time=0.2, value=1}
         {arriveTangent=-1.25, interpMode=RCIM_Cubic, leaveTangent=-1.25, time=0.7, value=1}
         <at 1 0>
        ]}}})
    Scale RGB=&VectorFromFloat(Value=&FloatFromCurve(FloatCurve=DataInterface() {propertyValues {Curve {keys [
          {arriveTangent=-4.5825, interpMode=RCIM_Cubic, leaveTangent=-4.5825, tangentMode=RCTM_User, time=0, value=1}
          {arriveTangent=-1, interpMode=RCIM_Cubic, leaveTangent=-1, time=0.1442, value=0.3607}
          <at 1 0>
         ]}}}))
    ScaleA=true
    ScaleRGB=true
   ) {enabled=false}
   &ScaleColor(
    Scale Alpha=&FloatFromCurve(FloatCurve=DataInterface() {propertyValues {Curve {keys [
         <at 0 0 RCIM_Cubic>
         {arriveTangent=1.4286, interpMode=RCIM_Cubic, leaveTangent=1.4286, time=0.2, value=1}
         {arriveTangent=-1.25, interpMode=RCIM_Cubic, leaveTangent=-1.25, time=0.7, value=1}
         <at 1 0>
        ]}}})
    ScaleA=true
    ScaleRGB=false
   ) {enabled=false, module=ScaleColor001}
   &UpdateMeshOrientation(
    Rotation Rate=&FloatFromCurve(FloatCurve=DataInterface() {propertyValues {Curve {keys [<at 0 1>,<at 0.75 0.1>]}}}, CurveIndex=$Emitter.Age, Scale Curve=10)
    Rotation Vector=$Particles.SpawnRotationVector
   ) {enabled=false, module=UpdateMeshOrientation001}
   &ScaleMeshSize(Scale Factor=&VectorFromFloat(Value=&FloatFromCurve(
      FloatCurve=DataInterface() {propertyValues {Curve {keys [<at 0 0 RCIM_Cubic>, <at 0.2 1>, {arriveTangent=-1.25, interpMode=RCIM_Cubic, leaveTangent=-1.25, time=0.8, value=1}, <at 1 0>]}}}
      Scale Curve=&FloatFromCurve(
       FloatCurve=DataInterface() {propertyValues {Curve {keys [{interpMode=RCIM_Cubic, tangentMode=RCTM_Break, time=0, value=1}, <at 60000 8>, <at 70000 0>]}}}
       CurveIndex=$Engine.Owner.LODDistance
      )
     ))) {enabled=false}
   &ScaleSpriteSize(Scale Factor=&Vector2DFromCurve(
     Vector2Curve=DataInterface() {propertyValues {
       XCurve {keys [{arriveTangent=-8.0902, interpMode=RCIM_Cubic, time=0, value=1}, <at 1 0 RCIM_Cubic>]}
       YCurve {keys [
         {arriveTangent=4.4388, interpMode=RCIM_Cubic, time=0, value=1}
         {arriveTangent=-29.2162, interpMode=RCIM_Cubic, leaveTangent=-29.2163, tangentMode=RCTM_User, time=0.2, value=6}
         <at 0.4 0.5>
         <at 1 0 RCIM_Cubic>
        ]}
      }}
     ScaleCurve=<1 1>
    ))
   /SetVariables(Particles.BurstSize:Vector2f=&Vector2DFromFloat(Value=&FloatFromCurve(
      FloatCurve=DataInterface() {propertyValues {Curve {keys [<at 0 1>,<at 0.2 0>]}}}
      Scale Curve=&RandomRangeFloat(Minimum=75, Maximum=150)
     ))) {isSetParameters=true}
  ]
  rendererClasses [&NiagaraMeshRendererProperties, &NiagaraSpriteRendererProperties]
  renderers [
   {rendererClass=&NiagaraMeshRendererProperties, rendererIndex=0}
   {rendererClass=&NiagaraSpriteRendererProperties, rendererIndex=1}
   {rendererClass=&NiagaraSpriteRendererProperties, rendererIndex=2}
  ]
 }
}
summary {
 systemName=NS_TowerObelisk_LootDrop
 emitters [
  {emitterName=CoronaBurst, rendererClasses [&NiagaraSpriteRendererProperties]}
  {
   bEnabled=false
   emitterName=Empty
   rendererClasses [&NiagaraSpriteRendererProperties, &NiagaraMeshRendererProperties]
  }
  {bEnabled=false, emitterName=Empty001, rendererClasses [&NiagaraSpriteRendererProperties]}
  {bEnabled=false, emitterName=Reachers, rendererClasses [&NiagaraRibbonRendererProperties]}
  {emitterName=LightSpike, rendererClasses [&NiagaraSpriteRendererProperties]}
  {emitterName=TinyFrags, rendererClasses [&NiagaraMeshRendererProperties]}
  {emitterName=Init_Glow, rendererClasses [&NiagaraSpriteRendererProperties]}
  {bEnabled=false, emitterName=Init_Glow001, rendererClasses [&NiagaraSpriteRendererProperties]}
  {emitterName=TinyFragsDeath, rendererClasses [&NiagaraMeshRendererProperties, &NiagaraSpriteRendererProperties]}
  {bEnabled=false, emitterName=TinyFrags002, rendererClasses [&NiagaraMeshRendererProperties]}
  {emitterName=Embers, simTarget=GPUComputeSim, rendererClasses [&NiagaraSpriteRendererProperties]}
  {bEnabled=false, emitterName=FN_Sprite, simTarget=GPUComputeSim, rendererClasses [&NiagaraSpriteRendererProperties]}
  {emitterName=CoronaLinger, rendererClasses [&NiagaraSpriteRendererProperties]}
 ]
 userVariables [User.PlayerLocation {type=Vector3f, description="", defaultValue {struct=Vector3f, value=<518.0515 20 75>}}]
}
