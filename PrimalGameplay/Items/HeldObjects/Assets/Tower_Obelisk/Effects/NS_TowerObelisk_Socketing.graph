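# Niagara effect for socketing on the Tower Obelisk: a roughly 3.5-second inward-collapsing charge-up.
# An initial glow flash opens it, alongside a flickering corona, a pulsing, shrinking dark ring and accelerating ring pulses, then a brief shrinking burst closes it.
# Streaks and mesh shards converge on the owner by curve-driven position lerps rather than velocity, with age curves tightening the streaks over time.
# Two outward-velocity emitters, Streaks001 and TinyFrags001, are disabled.
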NiagaraSystem /PrimalGameplay/Items/HeldObjects/Assets/Tower_Obelisk/Effects/NS_TowerObelisk_Socketing
dependencies {
 usedDataInterfaces [
  /Script/Niagara.NiagaraDataInterfaceSimpleCounter
  &NiagaraDataInterfaceCurve
  &NiagaraDataInterfaceVector2DCurve
 ]
 usedDynamicInputs [
  &VectorLength
  &Multiply_Vector
  &RotateVectorByQuaternion
  &NormalizeVector
 ]
 usedModules [
  &self:SystemSpawnScript.SystemScriptSource.SystemScriptGraph.NiagaraNodeAssignment_1.SetVariables_#1
  &SystemState
  &self:SystemSpawnScript.SystemScriptSource.SystemScriptGraph.NiagaraNodeAssignment_0.SetVariables_#2
  &self:SystemSpawnScript.SystemScriptSource.SystemScriptGraph.NiagaraNodeAssignment_2.SetVariables_#3
 ]
}
emitters {
 Corona {
  emitterUpdateScript [
   &EmitterState()
   &SpawnBurst_Instantaneous(Spawn Count=5, Use Spawn Probability=false)
   &SpawnRate(SpawnGroup=1, SpawnRate=20, Use Spawn Probability=false)
  ]
  particleSpawnScript [
   /SetVariables(
    Particles.Counter:NiagaraInt32=/Niagara/DynamicInputs/DataInterface/GenerateIntFromCounter(SimpleCounter:NiagaraDataInterfaceSimpleCounter=DataInterface() {propertyValues {GpuSyncMode=None, InitialValue=0}})
    Particles.SpawnGroup=&SpawnGroup()
   ) {isSetParameters=true}
   &InitializeParticle(
    AdjustAlpha=true
    AdjustHue=true
    AdjustSaturation=true
    AdjustValue=false
    Alpha Scale Range:Vector2f=<0.05 0.2>
    Color=<rgba 1 0.63 0.63 1>
    Color Mode=::NewEnumerator3
    Hue Shift Range:Vector2f=<-1 1>
    Lifetime Max=0.8
    Lifetime Min=0.4
    Lifetime Mode=::NewEnumerator1
    Saturation Range:Vector2f=<0.75 1>
    Sprite Rotation Mode=::NewEnumerator2
    Sprite Size Mode=::NewEnumerator3
    Uniform Sprite Size=&MakeCustomFloatFromBool(
     Bool=&SetBoolByIntComparison(A=$Particles.SpawnGroup, B=0, Comparison Type=::NewEnumerator4)
     True Float=&RandomRangeFloat(Minimum=175, Maximum=400)
     False Float=&RandomRangeFloat(Minimum=250, Maximum=300)
    )
    UsePositionOffset=false
    Write Material Random=true
   )
  ]
  particleUpdateScript [
   &ParticleState()
   &ScaleSpriteSize(Scale Factor=&Vector2DFromFloat(Value=&MakeCustomFloatFromBool(
      Bool=&SetBoolByIntComparison(A=$Particles.SpawnGroup, B=0, Comparison Type=::NewEnumerator4)
      True Float=&FloatFromCurve(FloatCurve=DataInterface() {propertyValues {Curve {keys [{arriveTangent=6.4229, interpMode=RCIM_Cubic, leaveTangent=6.4229, tangentMode=RCTM_User, time=0, value=0}, <at 0.2 1 RCIM_Cubic>]}}})
      False Float=&FloatFromCurve(FloatCurve=DataInterface() {propertyValues {Curve {keys [<at 0 1>,<at 3 0.6>]}}}, CurveIndex=$Engine.System.Age)
     )))
   &DynamicMaterialParameters(
    Index 0 Param 1=&RandomRangeFloat(Minimum=0.75, Maximum=1.5)
    Index 0 Param 2=&RandomRangeFloat(Minimum=-0.4, Maximum=0.4)
    Index 0 Param 3=1
    Index 0 Param 4=1
    Index 1 Param 1=&RandomRangeFloat(Minimum=3.2, Maximum=5)
    Index 1 Param 2=&FloatFromCurve(
     FloatCurve=DataInterface() {propertyValues {Curve {keys [
         {arriveTangent=10.7118, interpMode=RCIM_Cubic, leaveTangent=10.7118, tangentMode=RCTM_User, time=0, value=0}
         {interpMode=RCIM_Cubic, tangentMode=RCTM_User, time=0.2, value=1}
         <at 1 0>
        ]}}}
     Scale Curve=&RandomRangeFloat(Minimum=0.1, Maximum=0.7)
    )
    Index 1 Param 3=2
    Module.Param0WriteEnabled=true
    Module.Param0WriteEnabled1=true
    Module.Param1WriteEnabled=true
    Module.Param1WriteEnabled1=true
    Module.Param2WriteEnabled=true
    Module.Param2WriteEnabled1=true
    Module.Param3WriteEnabled=true
    Module.Param3WriteEnabled1=true
    Write Parameter Index 1=true
   )
   &ScaleColor(
    Scale Alpha=&FloatFromCurve(
     FloatCurve=DataInterface() {propertyValues {Curve {keys [
         {arriveTangent=2.5865, interpMode=RCIM_Cubic, leaveTangent=2.5865, tangentMode=RCTM_User, time=0, value=0}
         {arriveTangent=-0.0344, interpMode=RCIM_Cubic, leaveTangent=-0.0344, tangentMode=RCTM_User, time=0.5, value=1}
         {arriveTangent=-0.2386, interpMode=RCIM_Cubic, leaveTangent=-0.2386, tangentMode=RCTM_User, time=1.0011, value=0.002}
        ]}}}
     Scale Curve=&Lerp_Float(A=1, B=0.75, Alpha=$Engine.System.Age)
    )
    ScaleA=true
    ScaleRGB=false
   )
   &SpriteRotationRate(Rotation Rate=&Lerp_Float(A=&RandomRangeFloat(Minimum=-75, Maximum=75), B=0, Alpha=$Particles.NormalizedAge))
   &CameraOffset(Camera Offset Amount=20)
   &KillParticles(Kill Particles=&SetBoolByFloatComparison(A=$System.Age, B=3.4, Comparison Type=::NewEnumerator2)) {enabled=false}
  ]
  rendererClasses [&NiagaraSpriteRendererProperties]
  renderers [{rendererClass=&NiagaraSpriteRendererProperties}]
 }
 DarkRing {
  emitterUpdateScript [
   &EmitterState()
   &SpawnBurst_Instantaneous(Spawn Count=1, Use Spawn Probability=false)
   /SetVariables(Emitter.Pulse:NiagaraFloat=&FloatFromCurve(
     FloatCurve=DataInterface() {propertyValues {Curve {keys [
         {interpMode=RCIM_Cubic, tangentMode=RCTM_User, time=0, value=1}
         {interpMode=RCIM_Cubic, tangentMode=RCTM_User, time=0.1867, value=0.8133}
         {interpMode=RCIM_Cubic, tangentMode=RCTM_User, time=0.3711, value=0.6289}
         {interpMode=RCIM_Cubic, tangentMode=RCTM_User, time=0.5409, value=0.4591}
         {interpMode=RCIM_Cubic, tangentMode=RCTM_User, time=0.7169, value=0.2831}
         {interpMode=RCIM_Cubic, tangentMode=RCTM_User, time=0.8552, value=0.1448}
         <at 1 0 RCIM_Cubic>
        ]}}}
     CurveIndex=&Multiply_Float(A=$Engine.System.Age, B=0.2857)
    )) {enabled=false, isSetParameters=true}
   /SetVariables(Emitter.Pulse:NiagaraFloat=&FloatFromCurve(
     FloatCurve=DataInterface() {propertyValues {Curve {keys [
         {interpMode=RCIM_Cubic, tangentMode=RCTM_User, time=0, value=1}
         {arriveTangent=-0.3199, interpMode=RCIM_Cubic, leaveTangent=-0.3199, time=0.1706, value=1.0331}
         {interpMode=RCIM_Cubic, tangentMode=RCTM_User, time=0.2812, value=0.91}
         {arriveTangent=-0.4006, interpMode=RCIM_Cubic, leaveTangent=-0.4006, time=0.4339, value=0.9322}
         {interpMode=RCIM_Cubic, tangentMode=RCTM_User, time=0.5309, value=0.81}
         {arriveTangent=-0.1106, interpMode=RCIM_Cubic, leaveTangent=-0.1106, tangentMode=RCTM_User, time=0.6307, value=0.8356}
         {interpMode=RCIM_Cubic, tangentMode=RCTM_User, time=0.7122, value=0.7526}
         {arriveTangent=-0.486, interpMode=RCIM_Cubic, leaveTangent=-0.486, tangentMode=RCTM_User, time=0.7929, value=0.7573}
         {interpMode=RCIM_Cubic, tangentMode=RCTM_User, time=0.846, value=0.6876}
         {arriveTangent=-0.4742, interpMode=RCIM_Cubic, leaveTangent=-0.4742, time=0.9097, value=0.6727}
         {interpMode=RCIM_Cubic, tangentMode=RCTM_User, time=0.9328, value=0.6465}
         {arriveTangent=-0.6911, interpMode=RCIM_Cubic, leaveTangent=-0.6911, tangentMode=RCTM_User, time=0.9812, value=0.6351}
         {interpMode=RCIM_Cubic, tangentMode=RCTM_User, time=1, value=0.6}
        ]}}}
     CurveIndex=&Multiply_Float(A=$Engine.System.Age, B=0.2857)
    )) {enabled=false, isSetParameters=true}
   /SetVariables(Emitter.ExtremePulse:NiagaraFloat=&FloatFromCurve(
     FloatCurve=DataInterface() {propertyValues {Curve {keys [
         {arriveTangent=0.6535, interpMode=RCIM_Cubic, leaveTangent=0.6535, tangentMode=RCTM_User, time=0, value=0.9212}
         {interpMode=RCIM_Cubic, tangentMode=RCTM_User, time=0.2067, value=1}
         {arriveTangent=0.5447, interpMode=RCIM_Cubic, leaveTangent=0.5447, tangentMode=RCTM_User, time=0.3639, value=0.6}
         {arriveTangent=-1.8621, interpMode=RCIM_Cubic, leaveTangent=-1.8621, tangentMode=RCTM_User, time=0.5217, value=0.6546}
         {arriveTangent=0.9626, interpMode=RCIM_Cubic, leaveTangent=0.9626, tangentMode=RCTM_User, time=0.6639, value=0.3}
         {arriveTangent=-1.6607, interpMode=RCIM_Cubic, leaveTangent=-1.6607, tangentMode=RCTM_User, time=0.7583, value=0.3528}
         {arriveTangent=0.695, interpMode=RCIM_Cubic, leaveTangent=0.695, tangentMode=RCTM_User, time=0.8639, value=0.1}
         {arriveTangent=-0.737, interpMode=RCIM_Cubic, leaveTangent=-0.737, tangentMode=RCTM_User, time=0.9251, value=0.152}
         {arriveTangent=-1.8414, interpMode=RCIM_Cubic, leaveTangent=-1.8414, tangentMode=RCTM_User, time=1, value=0}
        ]}}}
     CurveIndex=&Multiply_Float(A=$Engine.System.Age, B=0.2857)
    )) {isSetParameters=true}
   /SetVariables(Emitter.ExtremePulse:NiagaraFloat=&FloatFromCurve(
     FloatCurve=DataInterface() {propertyValues {Curve {keys [
         {arriveTangent=0.6535, interpMode=RCIM_Cubic, leaveTangent=0.6535, tangentMode=RCTM_User, time=0, value=1}
         {interpMode=RCIM_Cubic, tangentMode=RCTM_User, time=0.098, value=1.0428}
         {arriveTangent=0.9658, interpMode=RCIM_Cubic, leaveTangent=0.9658, tangentMode=RCTM_User, time=0.2489, value=0.7837}
         {interpMode=RCIM_Cubic, tangentMode=RCTM_User, time=0.3724, value=0.8563}
         {arriveTangent=2.013, interpMode=RCIM_Cubic, leaveTangent=2.013, tangentMode=RCTM_User, time=0.4838, value=0.4728}
         {interpMode=RCIM_Cubic, tangentMode=RCTM_User, time=0.584, value=0.5868}
         {arriveTangent=1.6752, interpMode=RCIM_Cubic, leaveTangent=1.6752, tangentMode=RCTM_User, time=0.6914, value=0.2656}
         {interpMode=RCIM_Cubic, tangentMode=RCTM_User, time=0.7977, value=0.3485}
         {arriveTangent=2.0675, interpMode=RCIM_Cubic, leaveTangent=2.0675, tangentMode=RCTM_User, time=0.8686, value=0.162}
         {interpMode=RCIM_Cubic, tangentMode=RCTM_User, time=0.9354, value=0.2241}
         {arriveTangent=-3.3113, interpMode=RCIM_Cubic, leaveTangent=-3.3113, tangentMode=RCTM_User, time=1, value=0}
        ]}}}
     CurveIndex=&Multiply_Float(A=$Engine.System.Age, B=0.2857)
    )) {enabled=false, isSetParameters=true}
   /SetVariables(Emitter.ChillPulse:NiagaraFloat=&FloatFromCurve(
     FloatCurve=DataInterface() {propertyValues {Curve {keys [
         {arriveTangent=0.6535, interpMode=RCIM_Cubic, leaveTangent=0.6535, tangentMode=RCTM_User, time=0, value=0.9212}
         {arriveTangent=-0.6667, interpMode=RCIM_Cubic, leaveTangent=-0.6667, tangentMode=RCTM_User, time=0.2067, value=1}
         {arriveTangent=0.7142, interpMode=RCIM_Cubic, leaveTangent=0.7142, tangentMode=RCTM_User, time=0.3639, value=0.6756}
         {interpMode=RCIM_Cubic, tangentMode=RCTM_User, time=0.4664, value=0.7043}
         {arriveTangent=0.5853, interpMode=RCIM_Cubic, leaveTangent=0.5853, tangentMode=RCTM_User, time=0.6754, value=0.3313}
         {interpMode=RCIM_Cubic, tangentMode=RCTM_User, time=0.7307, value=0.3528}
         {interpMode=RCIM_Cubic, tangentMode=RCTM_User, time=0.8628, value=0.1305}
         {interpMode=RCIM_Cubic, tangentMode=RCTM_User, time=0.9128, value=0.1305}
         {arriveTangent=-1.8414, interpMode=RCIM_Cubic, leaveTangent=-1.8414, tangentMode=RCTM_User, time=1, value=0}
        ]}}}
     CurveIndex=&Multiply_Float(A=$Engine.System.Age, B=0.2857)
    )) {isSetParameters=true}
   /SetVariables(Emitter.ChillPulse:NiagaraFloat=&FloatFromCurve(
     FloatCurve=DataInterface() {propertyValues {Curve {keys [
         {arriveTangent=0.6535, interpMode=RCIM_Cubic, leaveTangent=0.6535, tangentMode=RCTM_User, time=0, value=0.9212}
         {interpMode=RCIM_Cubic, tangentMode=RCTM_User, time=0.2067, value=1}
         {arriveTangent=0.8236, interpMode=RCIM_Cubic, leaveTangent=0.8236, tangentMode=RCTM_User, time=0.3639, value=0.557}
         {interpMode=RCIM_Cubic, tangentMode=RCTM_User, time=0.4664, value=0.6182}
         {arriveTangent=0.9546, interpMode=RCIM_Cubic, leaveTangent=0.9546, tangentMode=RCTM_User, time=0.6639, value=0.3}
         {interpMode=RCIM_Cubic, tangentMode=RCTM_User, time=0.7307, value=0.3528}
         {arriveTangent=0.8117, interpMode=RCIM_Cubic, leaveTangent=0.8117, tangentMode=RCTM_User, time=0.8639, value=0.1}
         {interpMode=RCIM_Cubic, tangentMode=RCTM_User, time=0.9128, value=0.1305}
         {arriveTangent=-1.8414, interpMode=RCIM_Cubic, leaveTangent=-1.8414, tangentMode=RCTM_User, time=1, value=0}
        ]}}}
     CurveIndex=&Multiply_Float(A=$Engine.System.Age, B=0.2857)
    )) {enabled=false, isSetParameters=true}
  ]
  particleSpawnScript [&InitializeParticle(
    Color=<rgba 0 0 0 0.15>
    Lifetime=3.5
    Sprite Size Mode=::NewEnumerator3
    Uniform Sprite Size=350
    UsePositionOffset=false
    Write Material Random=true
   )]
  particleUpdateScript [
   &ParticleState()
   &DynamicMaterialParameters(
    Index 0 Param 1=&Lerp_Float(A=1, B=0, Alpha=$Emitter.ExtremePulse)
    Module.Param0WriteEnabled=true
    Module.Param1WriteEnabled=true
    Module.Param2WriteEnabled=false
    Module.Param3WriteEnabled=false
   )
   &DynamicMaterialParameters(
    Index 0 Param 1=&Lerp_Float(A=8, B=0.5, Alpha=$Emitter.ExtremePulse)
    Index 0 Param 2=&Lerp_Float(A=0, B=1, Alpha=$Particles.NormalizedAge)
    Module.Param0WriteEnabled=true
    Module.Param1WriteEnabled=true
    Module.Param2WriteEnabled=false
    Module.Param3WriteEnabled=false
   ) {enabled=false, module=DynamicMaterialParameters001}
   &ScaleColor(
    Scale Alpha=&FloatFromCurve(
     FloatCurve=DataInterface() {propertyValues {Curve {keys [
         {interpMode=RCIM_Cubic, tangentMode=RCTM_User, time=0.0155, value=1}
         {arriveTangent=-1.0157, interpMode=RCIM_Cubic, leaveTangent=-1.0157, tangentMode=RCTM_User, time=0.95, value=0.599}
         {arriveTangent=-16.4592, interpMode=RCIM_Cubic, leaveTangent=-16.4593, tangentMode=RCTM_User, time=1, value=0}
        ]}}}
     Scale Curve=&ClampFloat(Float=&OneMinusFloat(Float=$Emitter.ExtremePulse), Min=0.2)
    )
    ScaleA=true
    ScaleRGB=false
   )
   &ScaleSpriteSize(Scale Factor=&Vector2DFromFloat(Value=&Add_Float(A=0.4, B=&Multiply_Float(A=0.6, B=$Emitter.ChillPulse))))
  ]
  rendererClasses [&NiagaraSpriteRendererProperties]
  renderers [{rendererClass=&NiagaraSpriteRendererProperties}]
 }
 DarkRing001 {
  emitterUpdateScript [
   &EmitterState()
   &SpawnRate(
    SpawnRate=&Multiply_Float(
     A=1
     B=&FloatFromCurve(
      FloatCurve=DataInterface() {propertyValues {Curve {keys [<at 0 2 RCIM_Cubic>, {arriveTangent=10.4002, interpMode=RCIM_Cubic, leaveTangent=112198.8672, tangentMode=RCTM_User, time=1, value=7}]}}}
      CurveIndex=&Power(a=$Emitter.NormalizedLoopAge, b=2)
     )
    )
    Use Spawn Probability=false
   )
  ]
  particleSpawnScript [&InitializeParticle(
    Color=<rgba 1 1 1 0.2>
    Lifetime=&FloatFromCurve(
     FloatCurve=DataInterface() {propertyValues {Curve {keys [{arriveTangent=-2.2525, time=0, value=1}, {leaveTangent=-2.5167, time=1, value=0.2}]}}}
     CurveIndex=$Emitter.NormalizedLoopAge
    )
    Sprite Rotation Mode=::NewEnumerator2
    Sprite Size Mode=::NewEnumerator3
    Uniform Sprite Size=200
    UsePositionOffset=false
    Write Material Random=true
   )]
  particleUpdateScript [
   &ParticleState()
   &DynamicMaterialParameters(Write Parameter Index 0=false)
   &ScaleColor(
    Scale Alpha=&FloatFromCurve(FloatCurve=DataInterface() {propertyValues {Curve {keys [<at 0 1>, {arriveTangent=-24.9877, interpMode=RCIM_Cubic, leaveTangent=-24.9879, tangentMode=RCTM_User, time=1, value=0}]}}})
    ScaleA=true
    ScaleRGB=false
   )
   &ScaleSpriteSize(Scale Factor=&Vector2DFromFloat(Value=&FloatFromCurve(FloatCurve=DataInterface() {propertyValues {Curve {keys [<at 0 1>,<at 1 0>]}}})))
   &ScaleSpriteSize(Scale Factor=&Vector2DFromFloat(Value=&FloatFromCurve(FloatCurve=DataInterface() {propertyValues {Curve {keys [<at 0 1>,<at 1 2>]}}}, CurveIndex=$Emitter.NormalizedLoopAge))) {module=ScaleSpriteSize001}
  ]
  rendererClasses [&NiagaraSpriteRendererProperties]
  renderers [{rendererClass=&NiagaraSpriteRendererProperties}]
 }
 DarkRing002 {
  emitterUpdateScript [
   &EmitterState()
   &SpawnBurst_Instantaneous(Spawn Count=1, Spawn Time=3.45, Use Spawn Probability=false)
  ]
  particleSpawnScript [&InitializeParticle(Lifetime=0.2, Sprite Size Mode=::NewEnumerator3, Uniform Sprite Size=750, UsePositionOffset=false, Write Material Random=true)]
  particleUpdateScript [
   &ParticleState()
   &DynamicMaterialParameters(
    Index 0 Param 1=&Lerp_Float(A=1, B=0.8, Alpha=$Particles.NormalizedAge)
    Index 0 Param 3=1
    Module.Param0WriteEnabled=true
    Module.Param1WriteEnabled=true
    Module.Param2WriteEnabled=true
    Module.Param3WriteEnabled=true
   ) {module=DynamicMaterialParameters001}
   &ScaleSpriteSize(Scale Factor=&Vector2DFromFloat(Value=&FloatFromCurve(FloatCurve=DataInterface() {propertyValues {Curve {keys [<at 0 1 RCIM_Cubic>, {interpMode=RCIM_Cubic, leaveTangent=-1.8997, time=1, value=0}]}}})))
  ]
  rendererClasses [&NiagaraSpriteRendererProperties]
  renderers [{rendererClass=&NiagaraSpriteRendererProperties}]
 }
 GlassShard {
  emitterUpdateScript [
   &EmitterState(Life Cycle Mode=::NewEnumerator1, Loop Behavior=::NewEnumerator1, Loop Duration=2.5, UseLoopDelay=false)
   &SpawnBurst_Instantaneous(Spawn Count=12, Use Spawn Probability=false)
   &SpawnRate(SpawnGroup=1, SpawnRate=5, Use Spawn Probability=false)
  ]
  particleSpawnScript [
   /SetVariables(Particles.MeshOrientation=&MakeQuatFromAxisAngle(
     Axis=&RandomRangeVector(Minimum=<0 0 0>, Maximum=<1 1 1>, Evaluation Type=::NewEnumerator1)
     AngleInDegrees=90
    )) {isSetParameters=true, module=SetVariables_#4}
   /SetVariables(Particles.SpawnRotationVector=&RandomRangeVector(Minimum=<-0.15 -0.15 -0.15>, Maximum=<0.15 0.15 0.15>, Evaluation Type=::NewEnumerator1)) {isSetParameters=true, module=SetVariables_#5}
   /Niagara/Modules/Spawn/Initialization/InitializeParticle(
    Color=&RandomRangeLinearColor(Minimum=<rgba 0 0 0 1>, Maximum=<rgba 6 6 6 1>)
    Lifetime=&RandomRangeFloat(Minimum=0.25, Maximum=2)
    Mesh Scale=&VectorFromFloat(Value=&Lerp_Float(
      A=0.75
      B=&MaskFloatBySpawnGroup(Masked Float=4, Passthrough Float=1.5)
      Alpha=&Power(a=&RandomRangeFloat(Minimum=0, Maximum=1), b=3)
     ))
    Module.Write Color=true
    Module.Write Lifetime=true
    Module.Write Mass=true
    Module.Write Material Random=true
    Module.Write Position=false
    Module.Write Scale=true
    Module.Write SpriteRotation=false
    Module.Write SpriteSize=false
   )
   &Color(Color=&MakeLinearColorFromVectorAndFloat("Vector (RGB)"=<9.3572 4.2143 10>, "Float (Alpha)"=&ClampFloat(Float=1))) {enabled=false}
   /Niagara/Modules/Spawn/Location/SphereLocation(
    Hemisphere X=true
    MaskSpawnGroup=false
    Module.Override Randomness=false
    Sphere Origin:NiagaraPosition=$Engine.Owner.Position
    Sphere Radius=&RandomRangeFloat(Minimum=10, Maximum=200, Evaluation Type=::NewEnumerator1)
    Surface Only=true
   )
  ]
  particleUpdateScript [
   &ParticleState()
   &ScaleColor(
    Scale Alpha=&FloatFromCurve(FloatCurve=DataInterface() {propertyValues {Curve {keys [
         <at 0 0 RCIM_Cubic>
         {arriveTangent=1.4286, interpMode=RCIM_Cubic, leaveTangent=1.4286, time=0.2, value=1}
         {arriveTangent=-1.25, interpMode=RCIM_Cubic, leaveTangent=-1.25, time=0.7, value=1}
         <at 1 0>
        ]}}})
    ScaleA=true
    ScaleRGB=false
   )
   &UpdateMeshOrientation(
    Rotation Rate=&FloatFromCurve(
     FloatCurve=DataInterface() {propertyValues {Curve {keys [
         {interpMode=RCIM_Cubic, tangentMode=RCTM_User, time=0, value=1}
         {arriveTangent=-2.0445, interpMode=RCIM_Cubic, leaveTangent=-2.0445, tangentMode=RCTM_User, time=0.75, value=0.1}
        ]}}}
     CurveIndex=$Emitter.Age
     Scale Curve=10
    )
    Rotation Vector=$Particles.SpawnRotationVector
   ) {module=UpdateMeshOrientation001}
   &ScaleMeshSize(Scale Factor=&VectorFromFloat(Value=&MaskFloatBySpawnGroup(
      Masked Float=&FloatFromCurve(FloatCurve=DataInterface() {propertyValues {Curve {keys [<at 0 0 RCIM_Cubic>, <at 0.2 1>, {arriveTangent=-1.25, interpMode=RCIM_Cubic, leaveTangent=-1.25, time=0.8, value=1}, <at 1 0>]}}})
      Passthrough Float=&FloatFromCurve(FloatCurve=DataInterface() {propertyValues {Curve {keys [<at 0 0 RCIM_Cubic>, <at 0.4962 1>, {arriveTangent=-1.9848, interpMode=RCIM_Cubic, leaveTangent=-1.9848, time=0.8, value=1}, <at 1 0>]}}})
     )))
   /SetVariables(Particles.Position=&Lerp_Vector(
     A=Linked() {linkedVariable {name=Particles.Initial.Position, type=NiagaraPosition}}
     B=Linked() {linkedVariable {name=Engine.Owner.Position, type=NiagaraPosition}}
     Alpha=&MakeCustomFloatFromBool(
      Bool=&SetBoolByFloatComparison(A=$Particles.Lifetime, B=0.5)
      True Float=&FloatFromCurve(FloatCurve=DataInterface() {propertyValues {Curve {keys [
           {interpMode=RCIM_Cubic, tangentMode=RCTM_User, time=0, value=0}
           {arriveTangent=0.3915, interpMode=RCIM_Cubic, leaveTangent=0.3915, tangentMode=RCTM_User, time=0.7692, value=0.0765}
           {arriveTangent=6.5661, interpMode=RCIM_Cubic, leaveTangent=6.5661, tangentMode=RCTM_User, time=1, value=1}
          ]}}})
      False Float=&FloatFromCurve(FloatCurve=DataInterface() {propertyValues {Curve {keys [
           {interpMode=RCIM_Cubic, tangentMode=RCTM_User, time=0, value=0}
           {arriveTangent=0.3915, interpMode=RCIM_Cubic, leaveTangent=0.3915, tangentMode=RCTM_User, time=0.5875, value=0.0765}
           {arriveTangent=2.9342, interpMode=RCIM_Cubic, leaveTangent=2.9342, tangentMode=RCTM_User, time=1, value=1}
          ]}}})
     )
    )) {isSetParameters=true}
  ]
  rendererClasses [&NiagaraMeshRendererProperties]
  renderers [{rendererClass=&NiagaraMeshRendererProperties, rendererIndex=0}]
 }
 Init_Glow {
  emitterUpdateScript [
   &EmitterState()
   &SpawnBurst_Instantaneous(Spawn Count=1, Use Spawn Probability=false)
  ]
  particleSpawnScript [
   &InitializeParticle(
    Color=<rgba 0 0.5128 2 1>
    Lifetime=0.2
    Module.Write Color=true
    Module.Write Lifetime=true
    Module.Write Mass=true
    Module.Write Material Random=true
    Module.Write Position=true
    Module.Write Scale=false
    Module.Write SpriteRotation=true
    Module.Write SpriteSize=true
    Sprite Rotation=&RandomRangeFloat(Minimum=0, Maximum=360)
    Sprite Size=&Vector2DFromFloat(Value=800)
   )
   /SetVariables(Particles.Color2=<rgba 1.7896 0.3291 2.5 3>) {isSetParameters=true}
   /SetVariables(Particles.SpriteSize2=&Vector2DFromFloat(Value=500)) {isSetParameters=true}
   &DynamicMaterialParameters(
    Index 0 Param 1=0.7
    Index 0 Param 2=0.2
    Index 0 Param 3=1
    Index 0 Param 4=1
    Module.Param0WriteEnabled=true
    Module.Param1WriteEnabled=true
    Module.Param2WriteEnabled=true
    Module.Param3WriteEnabled=true
   )
  ]
  particleUpdateScript [
   &ParticleState()
   &ScaleColor(
    Scale Alpha=&FloatFromCurve(FloatCurve=DataInterface() {propertyValues {Curve {keys [{arriveTangent=-2.0437, interpMode=RCIM_Cubic, leaveTangent=-2.0437, tangentMode=RCTM_User, time=0, value=1.0115}, <at 1 0>]}}})
    ScaleA=true
    ScaleRGB=false
   )
   /SetVariables(Particles.Color2=&MultiplyLinearColor(
     Linear Color=$Particles.Initial.Color2
     Scale Factor=&MakeFloat4FromVectorAndFloat(
      "Vector (RGB)"=<1 1 1>
      "Float (Alpha)"=&FloatFromCurve(FloatCurve=DataInterface() {propertyValues {Curve {keys [{arriveTangent=-1.9382, interpMode=RCIM_Cubic, leaveTangent=-1.9382, tangentMode=RCTM_User, time=0.1018, value=1.0115}, <at 1 0>]}}})
     )
    )) {isSetParameters=true}
   &ScaleColor(
    Color Value To Scale=$Particles.Color2
    Scale Alpha=&FloatFromCurve(FloatCurve=DataInterface() {propertyValues {Curve {keys [<at 0 0.5>,<at 1 0>]}}})
    ScaleA=true
    ScaleRGB=true
   ) {enabled=false, module=ScaleColor001}
   &ScaleSpriteSize(Scale Factor=&Vector2DFromFloat(Value=&FloatFromCurve(FloatCurve=DataInterface() {propertyValues {Curve {keys [
          {interpMode=RCIM_Cubic, leaveTangent=4.5689, tangentMode=RCTM_Break, time=0, value=0}
          {arriveTangent=-0.0921, interpMode=RCIM_Cubic, leaveTangent=-0.0921, tangentMode=RCTM_User, time=0.125, value=1}
          <at 1 0>
         ]}}})))
   &CameraOffset(Camera Offset Amount=100)
   &SpriteRotationRate(Rotation Rate=&Lerp_Float(A=30, B=0, Alpha=$Particles.NormalizedAge))
  ]
  rendererClasses [&NiagaraSpriteRendererProperties]
  renderers [
   {rendererClass=&NiagaraSpriteRendererProperties}
   {rendererClass=&NiagaraSpriteRendererProperties, rendererIndex=1}
  ]
 }
 Streaks {
  emitterUpdateScript [
   &EmitterState()
   &SpawnBurst_Instantaneous(Spawn Count=20, Use Spawn Probability=false)
   &SpawnRate(SpawnRate=75, Use Spawn Probability=false)
  ]
  particleSpawnScript [
   /Niagara/Modules/Spawn/Initialization/V2/InitializeParticle(
    AdjustAlpha=true
    AdjustHue=true
    AdjustSaturation=true
    AdjustValue=false
    Alpha Scale Range:Vector2f=<0.1 2>
    Color=<rgba 0.44 0 1 0.5>
    Color Mode=::NewEnumerator3
    Hue Shift Range:Vector2f=<-0.35 0.35>
    Lifetime=&FloatFromCurve(
     FloatCurve=DataInterface() {propertyValues {Curve {keys [
         {arriveTangent=-0.003, interpMode=RCIM_Cubic, leaveTangent=-0.003, tangentMode=RCTM_User, time=0, value=1}
         {arriveTangent=-0.0636, interpMode=RCIM_Cubic, leaveTangent=-0.0636, tangentMode=RCTM_User, time=3.2, value=0.2}
        ]}}}
     CurveIndex=$Engine.System.Age
     Scale Curve=0.25
    )
    Saturation Range:Vector2f=<0.3 0.5>
    Sprite Size=&Multiply_Vector2(
     A=&RandomRangeVector2D(Minimum=<1 20>, Maximum=<2 40>)
     B=&Vector2DFromCurve(
      Vector2Curve=DataInterface() {propertyValues {XCurve {keys [<at 0 1>,<at 3 2>]}, YCurve {keys [<at 0 1>,<at 3 0.25>]}}}
      CurveIndex=$Engine.System.Age
      ScaleCurve=<1 1>
     )
    )
    Sprite Size Mode=::NewEnumerator4
    UsePositionOffset=false
    Write Material Random=true
   )
   /Niagara/Modules/Spawn/Location/V2/SphereLocation(
    Hemisphere X=true
    MaskSpawnGroup=false
    Offset=&MakeVector(
     X=&FloatFromCurve(
      FloatCurve=DataInterface() {propertyValues {Curve {keys [
          {arriveTangent=-0.003, interpMode=RCIM_Cubic, leaveTangent=-0.003, tangentMode=RCTM_User, time=0, value=0}
          {arriveTangent=-0.2893, interpMode=RCIM_Cubic, leaveTangent=-0.2893, tangentMode=RCTM_User, time=3.2, value=-10}
         ]}}}
      CurveIndex=$Engine.System.Age
     )
     Y=0
     Z=0
    )
    Override Seed=false
    Sphere Radius=&FloatFromCurve(
     FloatCurve=DataInterface() {propertyValues {Curve {keys [
         {arriveTangent=-0.003, interpMode=RCIM_Cubic, leaveTangent=-0.003, tangentMode=RCTM_User, time=0, value=1}
         {arriveTangent=-0.4831, interpMode=RCIM_Cubic, leaveTangent=-0.4831, tangentMode=RCTM_User, time=3.2, value=0.35}
        ]}}}
     CurveIndex=$Engine.System.Age
     Scale Curve=250
    )
    Surface Only=false
    UseNonUniformScale=false
    UseOffset=true
    UseSphereOrientation=false
   )
   &DynamicMaterialParameters(
    Index 0 Param 1=&RandomRangeFloat(Minimum=0.5, Maximum=2)
    Index 0 Param 2=20
    Module.Param0WriteEnabled=true
    Module.Param1WriteEnabled=true
    Module.Param2WriteEnabled=false
    Module.Param3WriteEnabled=false
   )
  ]
  particleUpdateScript [
   &ParticleState()
   &ScaleColor(
    Scale Alpha=&FloatFromCurve(FloatCurve=DataInterface() {propertyValues {Curve {keys [<at 0 0 RCIM_Cubic>,<at 1 1 RCIM_Cubic>]}}})
    ScaleA=true
    ScaleRGB=false
   )
   &ScaleSpriteSize(Scale Factor=&Vector2DFromCurve(
     Vector2Curve=DataInterface() {propertyValues {XCurve {keys [<at 0.5 1>,<at 1 0>]}, YCurve {keys [<at 0 1>,<at 1 1>]}}}
     ScaleCurve=&Vector2DFromFloat(Value=&RandomRangeFloat(Minimum=1, Maximum=2))
    ))
   /SetVariables(Particles.SpriteAlignment=&Subtract_Vector(A=Linked() {linkedVariable {name=Particles.Position, type=NiagaraPosition}}, B=/Niagara/DynamicInputs/Helpers/SimulationPosition())) {isSetParameters=true}
   /SetVariables(Particles.Position=&Lerp_Vector(
     A=Linked() {linkedVariable {name=Particles.Initial.Position, type=NiagaraPosition}}
     B=Linked() {linkedVariable {name=Engine.Owner.Position, type=NiagaraPosition}}
     Alpha=&FloatFromCurve(FloatCurve=DataInterface() {propertyValues {Curve {keys [
          {interpMode=RCIM_Cubic, tangentMode=RCTM_User, time=0, value=0}
          {arriveTangent=0.3915, interpMode=RCIM_Cubic, leaveTangent=0.3915, tangentMode=RCTM_User, time=0.5875, value=0.0765}
          {arriveTangent=2.9342, interpMode=RCIM_Cubic, leaveTangent=2.9342, tangentMode=RCTM_User, time=1, value=1}
         ]}}})
    )) {isSetParameters=true}
  ]
  rendererClasses [&NiagaraSpriteRendererProperties]
  renderers [{rendererClass=&NiagaraSpriteRendererProperties}]
 }
 Streaks001 {
  bEnabled=false
  emitterUpdateScript [
   &EmitterState()
   &SpawnBurst_Instantaneous(Spawn Count=20, Use Spawn Probability=false)
   &SpawnRate(SpawnRate=75, Use Spawn Probability=false)
  ]
  particleSpawnScript [
   &InitializeParticle(
    AdjustAlpha=true
    AdjustHue=true
    AdjustSaturation=true
    AdjustValue=false
    Alpha Scale Range:Vector2f=<0.1 2>
    Color=<rgba 0.44 0 1 0.5>
    Color Mode=::NewEnumerator3
    Hue Shift Range:Vector2f=<-0.35 0.35>
    Lifetime=&FloatFromCurve(
     FloatCurve=DataInterface() {propertyValues {Curve {keys [
         {arriveTangent=-0.003, interpMode=RCIM_Cubic, leaveTangent=-0.003, tangentMode=RCTM_User, time=0, value=1}
         {arriveTangent=-0.0636, interpMode=RCIM_Cubic, leaveTangent=-0.0636, tangentMode=RCTM_User, time=3.2, value=0.4}
        ]}}}
     CurveIndex=$Engine.System.Age
     Scale Curve=0.25
    )
    Saturation Range:Vector2f=<0.3 0.5>
    Sprite Size=&Multiply_Vector2(
     A=&RandomRangeVector2D(Minimum=<1 20>, Maximum=<2 40>)
     B=&Vector2DFromCurve(
      Vector2Curve=DataInterface() {propertyValues {XCurve {keys [<at 0 1>,<at 3 2>]}, YCurve {keys [<at 0 1>,<at 3 0.25>]}}}
      CurveIndex=$Engine.System.Age
      ScaleCurve=<1 1>
     )
    )
    Sprite Size Mode=::NewEnumerator4
    UsePositionOffset=false
    Write Material Random=true
   )
   &SphereLocation(
    Hemisphere X=true
    MaskSpawnGroup=false
    Offset=&MakeVector(
     X=&FloatFromCurve(
      FloatCurve=DataInterface() {propertyValues {Curve {keys [
          {arriveTangent=-0.003, interpMode=RCIM_Cubic, leaveTangent=-0.003, tangentMode=RCTM_User, time=0, value=0}
          {arriveTangent=-0.2893, interpMode=RCIM_Cubic, leaveTangent=-0.2893, tangentMode=RCTM_User, time=3.2, value=-0.5}
         ]}}}
      CurveIndex=$Engine.System.Age
      Scale Curve=80
     )
     Y=0
     Z=0
    )
    Override Seed=false
    Sphere Radius=&FloatFromCurve(
     FloatCurve=DataInterface() {propertyValues {Curve {keys [
         {arriveTangent=-0.003, interpMode=RCIM_Cubic, leaveTangent=-0.003, tangentMode=RCTM_User, time=0, value=1}
         {arriveTangent=-0.4831, interpMode=RCIM_Cubic, leaveTangent=-0.4831, tangentMode=RCTM_User, time=3.2, value=0.2}
        ]}}}
     CurveIndex=$Engine.System.Age
     Scale Curve=80
    )
    Surface Only=false
    UseNonUniformScale=false
    UseOffset=true
    UseSphereOrientation=false
   )
   &AddVelocityFromPoint(
    Module.Use Falloff Distance=false
    Origin Offset=&MakeVector(
     X=&FloatFromCurve(
      FloatCurve=DataInterface() {propertyValues {Curve {keys [
          {arriveTangent=-0.003, interpMode=RCIM_Cubic, leaveTangent=-0.003, tangentMode=RCTM_User, time=0, value=0}
          {arriveTangent=-0.2893, interpMode=RCIM_Cubic, leaveTangent=-0.2893, tangentMode=RCTM_User, time=3.2, value=-0.5}
         ]}}}
      CurveIndex=$Engine.System.Age
      Scale Curve=80
     )
     Y=0
     Z=0
    )
    Velocity Strength=&RandomRangeFloat(Minimum=1000, Maximum=2000)
   )
   &DynamicMaterialParameters(
    Index 0 Param 1=&RandomRangeFloat(Minimum=0.5, Maximum=2)
    Index 0 Param 2=20
    Module.Param0WriteEnabled=true
    Module.Param1WriteEnabled=true
    Module.Param2WriteEnabled=false
    Module.Param3WriteEnabled=false
   )
  ]
  particleUpdateScript [
   &Drag(Drag=10, Use Linear Drag=true, Use Rotational Drag=false)
   &SolveForcesAndVelocity(Clamp Velocity=false, Limit Acceleration=false)
   &ParticleState()
   &ScaleColor(
    Scale Alpha=&FloatFromCurve(FloatCurve=DataInterface() {propertyValues {Curve {keys [<at 0 0 RCIM_Cubic>,<at 1 1 RCIM_Cubic>]}}})
    ScaleA=true
    ScaleRGB=false
   )
   &ScaleSpriteSize(Scale Factor=&Vector2DFromCurve(
     Vector2Curve=DataInterface() {propertyValues {XCurve {keys [<at 0.5 1>,<at 1 0>]}, YCurve {keys [<at 0 1>,<at 1 1>]}}}
     ScaleCurve=&Vector2DFromFloat(Value=&RandomRangeFloat(Minimum=1, Maximum=2))
    ))
  ]
  rendererClasses [&NiagaraSpriteRendererProperties]
  renderers [{rendererClass=&NiagaraSpriteRendererProperties}]
 }
 TinyFrags001 {
  bEnabled=false
  emitterUpdateScript [
   &EmitterState(
    Enable Distance Culling=true
    MaxDistance=75000
    Module.bUseMaxDistance=true
    Module.bUseMinDistance=false
    Scalability Mode=::NewEnumerator1
   )
   &SpawnBurst_Instantaneous(Spawn Count=12, Use Spawn Probability=false)
  ]
  particleSpawnScript [
   /SetVariables(Particles.MeshOrientation=&MakeQuatFromAxisAngle(
     Axis=&RandomRangeVector(Minimum=<0 0 0>, Maximum=<1 1 1>, Evaluation Type=::NewEnumerator1)
     AngleInDegrees=90
    )) {isSetParameters=true, module=SetVariables_#4}
   /SetVariables(Particles.SpawnRotationVector=&RandomRangeVector(Minimum=<-0.15 -0.15 -0.15>, Maximum=<0.15 0.15 0.15>, Evaluation Type=::NewEnumerator1)) {isSetParameters=true, module=SetVariables_#5}
   /Niagara/Modules/Spawn/Initialization/InitializeParticle(
    Color=&RandomRangeLinearColor(Minimum=<rgba 0 0 0 1>, Maximum=<rgba 6 6 6 1>)
    Lifetime=&MakeCustomFloatFromBool(
     Bool=&SetBoolByFloatComparison(A=&ReturnNormalizedExecIndex(), B=0.2, Comparison Type=::NewEnumerator6)
     True Float=&RandomRangeFloat(Minimum=0.5, Maximum=1, Evaluation Type=::NewEnumerator1)
     False Float=&RandomRangeFloat(Minimum=0.2, Maximum=0.5, Evaluation Type=::NewEnumerator1)
    )
    Mesh Scale=&RandomRangeVector(Minimum=<0.75 0.75 0.75>, Maximum=<4 4 4>, Evaluation Type=::NewEnumerator1)
    Module.Write Color=true
    Module.Write Lifetime=true
    Module.Write Mass=true
    Module.Write Material Random=true
    Module.Write Position=false
    Module.Write Scale=true
    Module.Write SpriteRotation=false
    Module.Write SpriteSize=false
   )
   &Color(Color=&MakeLinearColorFromVectorAndFloat("Vector (RGB)"=<9.3572 4.2143 10>, "Float (Alpha)"=&ClampFloat(Float=1))) {enabled=false}
   /Niagara/Modules/Spawn/Location/SphereLocation(
    Hemisphere X=true
    MaskSpawnGroup=false
    Module.Override Randomness=false
    Sphere Radius=&RandomRangeFloat(Minimum=10, Maximum=60, Evaluation Type=::NewEnumerator1)
    Surface Only=true
   )
   &AddVelocityFromPoint(Module.Use Falloff Distance=false, Velocity Strength=&RandomRangeFloat(Minimum=50, Maximum=300))
  ]
  particleUpdateScript [
   &ParticleState()
   &SolveForcesAndVelocity(Clamp Velocity=false, Limit Acceleration=false)
   &ScaleColor(
    Scale Alpha=&FloatFromCurve(FloatCurve=DataInterface() {propertyValues {Curve {keys [
         <at 0 0 RCIM_Cubic>
         {arriveTangent=1.4286, interpMode=RCIM_Cubic, leaveTangent=1.4286, time=0.2, value=1}
         {arriveTangent=-1.25, interpMode=RCIM_Cubic, leaveTangent=-1.25, time=0.7, value=1}
         <at 1 0>
        ]}}})
    ScaleA=true
    ScaleRGB=false
   )
   /Niagara/Modules/Update/Orientation/V2/UpdateMeshOrientation(Rotation Vector:Vector3f=$Particles.SpawnRotationVector) {enabled=false}
   /Niagara/Modules/Update/Orientation/V3/UpdateMeshOrientation(
    Rotation Rate=&FloatFromCurve(FloatCurve=DataInterface() {propertyValues {Curve {keys [<at 0 1>,<at 0.75 0.1>]}}}, CurveIndex=$Emitter.Age, Scale Curve=10)
    Rotation Vector=$Particles.SpawnRotationVector
   ) {module=UpdateMeshOrientation001}
   &ScaleMeshSize(Scale Factor=&VectorFromFloat(Value=&FloatFromCurve(
      FloatCurve=DataInterface() {propertyValues {Curve {keys [<at 0 0 RCIM_Cubic>, <at 0.2 1>, {arriveTangent=-1.25, interpMode=RCIM_Cubic, leaveTangent=-1.25, time=0.8, value=1}, <at 1 0>]}}}
      Scale Curve=&FloatFromCurve(
       FloatCurve=DataInterface() {propertyValues {Curve {keys [{interpMode=RCIM_Cubic, tangentMode=RCTM_Break, time=0, value=1}, <at 60000 8>, <at 70000 0>]}}}
       CurveIndex=$Engine.Owner.LODDistance
      )
     )))
   /SetVariables(Particles.Velocity=&Multiply_VectorByFloat(
     Vector=$Particles.Initial.Velocity
     Float=&FloatFromCurve(
      FloatCurve=DataInterface() {propertyValues {Curve {keys [{arriveTangent=-2.5823, interpMode=RCIM_Cubic, leaveTangent=-2.5823, tangentMode=RCTM_User, time=0, value=1}, <at 0.8 0>]}}}
      CurveIndex=$Emitter.Age
     )
    )) {isSetParameters=true}
  ]
  rendererClasses [&NiagaraMeshRendererProperties]
  renderers [{rendererClass=&NiagaraMeshRendererProperties, rendererIndex=0}]
 }
}
summary {
 systemName=NS_TowerObelisk_Socketing
 emitters [
  {emitterName=Corona, rendererClasses [&NiagaraSpriteRendererProperties]}
  {emitterName=Streaks, rendererClasses [&NiagaraSpriteRendererProperties]}
  {emitterName=GlassShard, rendererClasses [&NiagaraMeshRendererProperties]}
  {emitterName=Init_Glow, rendererClasses [&NiagaraSpriteRendererProperties]}
  {bEnabled=false, emitterName=TinyFrags001, rendererClasses [&NiagaraMeshRendererProperties]}
  {bEnabled=false, emitterName=Streaks001, rendererClasses [&NiagaraSpriteRendererProperties]}
  {emitterName=DarkRing, rendererClasses [&NiagaraSpriteRendererProperties]}
  {emitterName=DarkRing001, rendererClasses [&NiagaraSpriteRendererProperties]}
  {emitterName=DarkRing002, rendererClasses [&NiagaraSpriteRendererProperties]}
 ]
 userVariables [
  User.bIsFrontEnd {type=NiagaraBool}
  User.bIsFrontEndPreview {type=NiagaraBool}
  User.Destabilization {type=NiagaraFloat, defaultValue=0, description=""}
  User.PlayerLocation {type=Vector3f, description="", defaultValue {struct=Vector3f, value=<518.0515 20 75>}}
 ]
}
