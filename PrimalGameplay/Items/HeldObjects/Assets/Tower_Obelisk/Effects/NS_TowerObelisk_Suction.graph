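# Niagara suction effect for the Tower Obelisk, drawing energy toward the system's origin and reacting to a supplied player location.
# Corona makes flickering glow sprites and distance-gated lights, while Streaks pulls aligned sprites inward.
# Reachers builds seeded Hermite-spline ribbons from the origin to the smoothed player position, retracting with age.
# System-level player distance and side values drive spawning, width and fading; four other emitters are disabled.
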NiagaraSystem /PrimalGameplay/Items/HeldObjects/Assets/Tower_Obelisk/Effects/NS_TowerObelisk_Suction
dependencies {
 usedDataInterfaces [
  /Script/Niagara.NiagaraDataInterfaceSimpleCounter
  &NiagaraDataInterfaceCurve
  &NiagaraDataInterfaceColorCurve
  &NiagaraDataInterfaceVector2DCurve
  &NiagaraDataInterfaceVectorCurve
 ]
 usedDynamicInputs [
  &VectorLength
  &DotProduct
  /Niagara/DynamicInputs/Helpers/SafeNormalizeVector
 ]
 usedModules [
  &self:SystemSpawnScript.SystemScriptSource.SystemScriptGraph.NiagaraNodeAssignment_1.SetVariables_#1
  &SystemState
  &self:SystemSpawnScript.SystemScriptSource.SystemScriptGraph.NiagaraNodeAssignment_0.SetVariables_#2
  &self:SystemSpawnScript.SystemScriptSource.SystemScriptGraph.NiagaraNodeAssignment_2.SetVariables_#3
  &self:SystemSpawnScript.SystemScriptSource.SystemScriptGraph.NiagaraNodeAssignment_3.SetVariables_#4
 ]
}
emitters {
 Corona {
  emitterUpdateScript [
   &EmitterState()
   &SpawnRate(
    Spawn Probability=&MakeCustomFloatFromBool(
     Bool=&SetBoolByFloatComparison(A=$System.DistanceToPlayer, B=1505, Comparison Type=::NewEnumerator6)
     True Float=1
     False Float=0
    )
    SpawnRate=5
    Use Spawn Probability=true
   )
  ]
  particleSpawnScript [
   &InitializeParticle(
    AdjustAlpha=true
    AdjustHue=true
    AdjustSaturation=true
    AdjustValue=false
    Alpha Scale Range:Vector2f=<0.05 0.1>
    Color=<rgba 1 0.63 0.63 1>
    Color Mode=::NewEnumerator3
    Hue Shift Range:Vector2f=<-1 1>
    Lifetime=2
    Saturation Range:Vector2f=<0.75 1>
    Sprite Rotation Mode=::NewEnumerator2
    Sprite Size Mode=::NewEnumerator1
    Uniform Sprite Size Max=225
    Uniform Sprite Size Min=175
    UsePositionOffset=false
    Write Material Random=true
   )
   &SystemLocation(MaskSpawnGroup=false, Offset=<60 0 0>)
   /SetVariables(Particles.Counter:NiagaraInt32=/Niagara/DynamicInputs/DataInterface/GenerateIntFromCounter(SimpleCounter:NiagaraDataInterfaceSimpleCounter=DataInterface() {propertyValues {GpuSyncMode=None, InitialValue=0}})) {isSetParameters=true}
  ]
  particleUpdateScript [
   &ParticleState()
   &ScaleSpriteSize(Scale Factor=&Vector2DFromFloat(Value=&FloatFromCurve(FloatCurve=DataInterface() {propertyValues {Curve {keys [
          {arriveTangent=2.6856, interpMode=RCIM_Cubic, leaveTangent=2.6856, tangentMode=RCTM_User, time=0, value=0.6449}
          {arriveTangent=0.3224, interpMode=RCIM_Cubic, leaveTangent=0.3224, time=0.1674, value=1.0088}
          {arriveTangent=-1.024, interpMode=RCIM_Cubic, leaveTangent=-1.024, tangentMode=RCTM_User, time=0.6257, value=0.8466}
          {arriveTangent=0.0432, interpMode=RCIM_Cubic, leaveTangent=0.0432, tangentMode=RCTM_User, time=1, value=0}
         ]}}})))
   &DynamicMaterialParameters(
    Index 0 Param 1=&RandomRangeFloat(Minimum=0.75, Maximum=1.5)
    Index 0 Param 2=&RandomRangeFloat(Minimum=-0.4, Maximum=0.4)
    Index 0 Param 3=1
    Index 0 Param 4=1
    Index 1 Param 1=&RandomRangeFloat(Minimum=3.2, Maximum=5)
    Index 1 Param 2=&FloatFromCurve(
     FloatCurve=DataInterface() {propertyValues {Curve {keys [
         {arriveTangent=10.7118, interpMode=RCIM_Cubic, leaveTangent=10.7118, tangentMode=RCTM_User, time=0, value=0}
         {interpMode=RCIM_Cubic, tangentMode=RCTM_User, time=0.2, value=1}
         <at 1 0>
        ]}}}
     Scale Curve=&RandomRangeFloat(Minimum=0.1, Maximum=0.7)
    )
    Index 1 Param 3=2
    Module.Param0WriteEnabled=true
    Module.Param0WriteEnabled1=true
    Module.Param1WriteEnabled=true
    Module.Param1WriteEnabled1=true
    Module.Param2WriteEnabled=true
    Module.Param2WriteEnabled1=true
    Module.Param3WriteEnabled=true
    Module.Param3WriteEnabled1=true
    Write Parameter Index 1=true
   )
   &DynamicMaterialParameters(
    Index 0 Param 1=&RandomRangeFloat(Minimum=0.75, Maximum=1.5)
    Index 0 Param 2=&RandomRangeFloat(Minimum=-0.4, Maximum=0.4)
    Index 0 Param 3=1
    Index 0 Param 4=1
    Index 1 Param 1=&RandomRangeFloat(Minimum=2.5, Maximum=5)
    Index 1 Param 2=&FloatFromCurve(
     FloatCurve=DataInterface() {propertyValues {Curve {keys [
         {arriveTangent=10.7118, interpMode=RCIM_Cubic, leaveTangent=10.7118, tangentMode=RCTM_User, time=0, value=0}
         {interpMode=RCIM_Cubic, tangentMode=RCTM_User, time=0.2, value=1}
         <at 1 0>
        ]}}}
     Scale Curve=&RandomRangeFloat(Minimum=0.1, Maximum=0.7)
    )
    Index 1 Param 3=2
    Module.Param0WriteEnabled=true
    Module.Param0WriteEnabled1=true
    Module.Param1WriteEnabled=true
    Module.Param1WriteEnabled1=true
    Module.Param2WriteEnabled=true
    Module.Param2WriteEnabled1=true
    Module.Param3WriteEnabled=true
    Module.Param3WriteEnabled1=true
    Write Parameter Index 1=true
   ) {enabled=false, module=DynamicMaterialParameters001}
   &ScaleColor(
    Scale Alpha=&FloatFromCurve(
     FloatCurve=DataInterface() {propertyValues {Curve {keys [
         <at 0 0 RCIM_Cubic>
         {arriveTangent=-0.3825, interpMode=RCIM_Cubic, leaveTangent=-0.3825, tangentMode=RCTM_User, time=0.1284, value=1}
         {arriveTangent=-0.2386, interpMode=RCIM_Cubic, leaveTangent=-0.2386, tangentMode=RCTM_User, time=1.0011, value=0.002}
        ]}}}
     Scale Curve=&ClampFloat(
      Float=&Add_Float(A=2, B=&Multiply_Float(A=$System.DistanceToPlayer, B=-0.0025))
      Max=2
     )
    )
    ScaleA=true
    ScaleRGB=false
   )
   &SpriteRotationRate(Rotation Rate=&RandomRangeFloat(Minimum=-4, Maximum=4))
   &CameraOffset(Camera Offset Amount=20)
   &FN_Light_Attributes(
    Color=&HueShiftLinearColor(
     Hue=0
     Saturation=2
     Value=1
     Linear Color=&MultiplyLinearColor(Linear Color=$Particles.Color, Scale Factor=<0.5 0.75 1 1>)
    )
    Create Color Attribute=true
    Create Exponent Attribute=true
    Create Position Attribute=true
    Create Radius Attribute=true
    Enable Particle Light Enable=true
    Exponent=3
    Light Enabled:NiagaraBool=&Boolean_AndOperation(
     Bool1=&SetBoolByFloatComparison(A=$System.DistanceToPlayer, B=1500, Comparison Type=::NewEnumerator6)
     Bool2=&SetBoolByIntComparison(A=&ModuloInt(A=$Particles.Counter, B=5), B=0, Comparison Type=::NewEnumerator4)
    )
    Position=&MakeCustomVectorFromBool(
     Bool=&SetBoolByIntComparison(A=&ModuloInt(A=$Particles.Counter, B=10), B=0, Comparison Type=::NewEnumerator4)
     True Vector=<20 -15 0>
     False Vector=&MakeCustomVectorFromBool(
      Bool=&SetBoolByIntComparison(A=&ModuloInt(A=$Particles.Counter, B=10), B=5, Comparison Type=::NewEnumerator4)
      True Vector=<20 10 0>
      False Vector=<0 0 0>
     )
    )
    Radius=50
    Scale Color=&FloatFromCurve(
     FloatCurve=DataInterface() {propertyValues {Curve {keys [
         {interpMode=RCIM_Cubic, tangentMode=RCTM_User, time=0, value=0}
         {interpMode=RCIM_Cubic, tangentMode=RCTM_User, time=0.25, value=1}
         <at 1 0>
        ]}}}
     Scale Curve=&Multiply_Float(
      A=&RandomRangeFloat(Minimum=4, Maximum=7)
      B=&ClampFloat(Float=&Add_Float(A=1.5, B=&Multiply_Float(A=$System.DistanceToPlayer, B=-0.001)))
     )
    )
   )
  ]
  rendererClasses [&NiagaraSpriteRendererProperties, &NiagaraLightRendererProperties]
  renderers [
   {rendererClass=&NiagaraSpriteRendererProperties}
   {rendererClass=&NiagaraLightRendererProperties, rendererIndex=1}
  ]
 }
 Empty {
  bEnabled=false
  emitterUpdateScript [
   &EmitterState()
   &SpawnRate(SpawnRate=5, Use Spawn Probability=false) {enabled=false}
   &SpawnBurst_Instantaneous(
    Spawn Count=1
    Spawn Probability=&MakeCustomFloatFromBool(
     Bool=&SetBoolByIntComparison(A=$Emitter.LoopCount, B=0, Comparison Type=::NewEnumerator4)
     True Float=1
     False Float=0
    )
    Use Spawn Probability=true
   )
  ]
  particleSpawnScript [
   &InitializeParticle(Color Mode=::NewEnumerator0, Mesh Scale Mode=::NewEnumerator3, UsePositionOffset=false, Write Material Random=true)
   &SystemLocation(MaskSpawnGroup=false, Offset=$User.PlayerLocation, Offset Coordinate Space=::Simulation) {enabled=false}
   &Color(Color=&ColorFromCurve(
     ColorCurve=DataInterface() {propertyValues {
       AlphaCurve {keys [<at 0 1>]}
       BlueCurve {keys [<at 0 0>,<at 0.5 0.0261>,<at 1.0012 1>]}
       GreenCurve {keys [<at 0 0>,<at 0.5 1>,<at 1.0012 0>]}
       RedCurve {keys [<at 0 1>,<at 0.5 0>,<at 1.0012 0>]}
      }}
     CurveIndex=&Multiply_Float(A=$System.DistanceToPlayer, B=0.001)
    )) {enabled=false}
   &KillParticles(Kill Particles=$System.bIsBehind)
  ]
  particleUpdateScript [
   &ParticleState(Kill Particles When Lifetime Has Elapsed=false, Let Infinitely Lived Particles Die When Emitter Deactivates=true)
   /Niagara/Modules/Update/Orientation/V2/OrientMeshToVector(
    Look At Direction:Vector3f=&Subtract_Vector(A=Linked() {linkedVariable {name=Engine.Owner.Position, type=NiagaraPosition}}, B=$User.PlayerLocation)
    Override Quat=false
    UseOptionalSideAxis=false
    UseOptionalUpAxis=false
   ) {enabled=false}
   /SetVariables(Particles.Position=&Add_Vector(
     A=$User.PlayerLocation
     B=&RotateVectorByQuaternion(
      Vector To Rotate=&RandomRangeVector(Minimum=<0 -20 -20>, Maximum=<0 20 20>)
      Quaternion=$Particles.MeshOrientation
     )
    )) {enabled=false, isSetParameters=true}
   /SetVariables(Particles.Position=&self:SmoothInterpolateParticleVector(
     Initial Vector:Vector3f=$User.PlayerLocation
     Vector to Interpolate:Vector3f=&Add_Vector(A=$User.PlayerLocation, B=<0 -25 0>)
     Rate Of Change:NiagaraFloat=7
    )) {enabled=false, isSetParameters=true}
   /SetVariables(Particles.Position:Vector3f=$User.PlayerLocation) {isSetParameters=true}
   &Color(Color=&ColorFromCurve(
     ColorCurve=DataInterface() {propertyValues {
       AlphaCurve {keys [<at 0 1 RCIM_Cubic>]}
       BlueCurve {keys [<at 0 0 RCIM_Cubic>,<at 0.4908 0>,<at 1 1 RCIM_Cubic>]}
       GreenCurve {keys [<at 0 0 RCIM_Cubic>,<at 0.4908 1>,<at 1 0 RCIM_Cubic>]}
       RedCurve {keys [<at 0 1 RCIM_Cubic>,<at 0.4908 0>,<at 1 0 RCIM_Cubic>]}
      }}
     CurveIndex=&MakeCustomFloatFromBool(Bool=$System.bIsBehind, True Float=1, False Float=0)
    )) {module=Color001}
   &Color(Color=&ColorFromCurve(
     ColorCurve=DataInterface() {propertyValues {
       AlphaCurve {keys [<at 0 1 RCIM_Cubic>]}
       BlueCurve {keys [<at 0 0 RCIM_Cubic>,<at 0.4908 0>,<at 1 1 RCIM_Cubic>]}
       GreenCurve {keys [<at 0 0 RCIM_Cubic>,<at 0.4908 1>,<at 1 0 RCIM_Cubic>]}
       RedCurve {keys [<at 0 1 RCIM_Cubic>,<at 0.4908 0>,<at 1 0 RCIM_Cubic>]}
      }}
     CurveIndex=$System.CompareSides
    )) {enabled=false, module=Color002}
  ]
  rendererClasses [&NiagaraSpriteRendererProperties, &NiagaraMeshRendererProperties]
  renderers [
   {rendererClass=&NiagaraSpriteRendererProperties}
   {rendererClass=&NiagaraMeshRendererProperties, rendererIndex=1}
   {rendererClass=&NiagaraMeshRendererProperties, rendererIndex=2}
  ]
 }
 Empty001 {
  bEnabled=false
  emitterUpdateScript [
   &EmitterState()
   &SpawnRate(SpawnRate=2, Use Spawn Probability=false)
  ]
  particleSpawnScript [
   &InitializeParticle(Color Mode=::NewEnumerator0, Sprite Size Mode=::NewEnumerator3, Uniform Sprite Size=500, UsePositionOffset=false, Write Material Random=true)
   &SystemLocation(MaskSpawnGroup=false, Offset=<-100 0 0>)
  ]
  particleUpdateScript [&ParticleState()]
  rendererClasses [&NiagaraSpriteRendererProperties]
  renderers [{rendererClass=&NiagaraSpriteRendererProperties}]
 }
 Empty002 {
  bEnabled=false
  emitterUpdateScript [
   &EmitterState()
   &SpawnRate(SpawnRate=100, Use Spawn Probability=false)
  ]
  particleSpawnScript [
   &InitializeParticle(
    AdjustAlpha=true
    AdjustHue=true
    AdjustSaturation=true
    AdjustValue=false
    Alpha Scale Range:Vector2f=<0.25 1>
    Color=<rgba 0.44 0 1 1>
    Color Mode=::NewEnumerator3
    Hue Shift Range:Vector2f=<-0.35 0.35>
    Lifetime Max=0.25
    Lifetime Min=0.1
    Lifetime Mode=::NewEnumerator1
    Saturation Range:Vector2f=<0.3 0.5>
    Sprite Size Max=<2 40>
    Sprite Size Min=<1 20>
    Sprite Size Mode=::NewEnumerator2
    UsePositionOffset=false
    Write Material Random=true
   )
   &AddVelocity(Velocity=<100 0 0>) {enabled=false}
   &SphereLocation(
    MaskSpawnGroup=false
    Offset=<40 0 0>
    Override Seed=false
    Sphere Radius=40
    Surface Only=false
    UseNonUniformScale=false
    UseOffset=true
    UseSphereOrientation=false
   )
   &AddVelocityInCone(
    Cone Angle=20
    Module.Use Velocity Falloff On Cone Axis=false
    Velocity Strength=&RandomRangeFloat(Minimum=30, Maximum=100)
   )
   &AddVelocityFromPoint(Module.Use Falloff Distance=false, Velocity Strength=75)
  ]
  particleUpdateScript [
   &SolveForcesAndVelocity(Clamp Velocity=false, Limit Acceleration=false)
   &ParticleState()
   &ScaleColor(
    Scale Alpha=&FloatFromCurve(FloatCurve=DataInterface() {propertyValues {Curve {keys [<at 0 0 RCIM_Cubic>,<at 0.5 1 RCIM_Cubic>]}}})
    ScaleA=true
    ScaleRGB=false
   )
   /SetVariables(Particles.Velocity=&Add_Vector(A=$Particles.Velocity, B=$Particles.Initial.Velocity)) {isSetParameters=true}
   &ScaleSpriteSize(Scale Factor=&Vector2DFromCurve(Vector2Curve=DataInterface() {propertyValues {XCurve {keys [<at 0.5 1>,<at 1 0>]}, YCurve {keys [<at 0 1>,<at 1 1>]}}}, ScaleCurve=<1 1>))
  ]
  rendererClasses [&NiagaraSpriteRendererProperties]
  renderers [{rendererClass=&NiagaraSpriteRendererProperties}]
 }
 IdleTentacles {
  bEnabled=false
  emitterSpawnScript [/SetVariables(
    Emitter.EmitterRandom:NiagaraInt32=/Niagara/DynamicInputs/Random/RandomInt(Randomness Mode:ENiagaraRandomnessMode=::NewEnumerator0, Random Seed:NiagaraInt32=0, Fixed Random Seed:NiagaraBool=false)
    Emitter.ParticlesPerRibbon=3
   ) {isSetParameters=true, module=SetVariables_#5}]
  emitterUpdateScript [
   &EmitterState(Life Cycle Mode=::NewEnumerator1, Loop Duration=0.1, UseLoopDelay=false)
   &SpawnBurst_Instantaneous(
    Spawn Count=$Emitter.ParticlesPerRibbon
    Spawn Probability=&MakeCustomFloatFromBool(Bool=&SetBoolByFloatComparison(A=$System.DistanceToPlayer, B=1005), True Float=0, False Float=1)
    Use Spawn Probability=true
   )
  ]
  particleSpawnScript [
   /SetVariables(Particles.RandomSeedByRibbon:NiagaraInt32=&Add_Integer(A=$Emitter.LoopCount, B=$Emitter.EmitterRandom)) {isSetParameters=true, module=SetVariables_#6}
   &OrientMeshToVector(
    Look At Direction:Vector3f=&Subtract_Vector(A=Linked() {linkedVariable {name=Engine.Owner.Position, type=NiagaraPosition}}, B=$User.PlayerLocation)
    Override Quat=false
    UseOptionalSideAxis=false
    UseOptionalUpAxis=false
   )
   &InitializeParticle(
    AdjustAlpha=true
    AdjustHue=true
    AdjustSaturation=true
    AdjustValue=false
    Alpha Scale Range:Vector2f=<0.15 0.3>
    Color=<rgba 0.44 0 1 0.5>
    Color Mode=::NewEnumerator3
    Hue Shift Range:Vector2f=<-0.35 0.35>
    Lifetime=&Lerp_Float(
     A=&RandomRangeFloat(Minimum=0.1, Maximum=0.5, Randomness Mode=::NewEnumerator1, Random Seed=$Particles.RandomSeedByRibbon, Fixed Random Seed=true)
     B=&RandomRangeFloat(Minimum=1, Maximum=2, Randomness Mode=::NewEnumerator1, Random Seed=$Particles.RandomSeedByRibbon, Fixed Random Seed=true)
     Alpha=$System.NormalizedDistance
    )
    Saturation Range:Vector2f=<0.3 0.5>
    Sprite Size Mode=::NewEnumerator3
    Uniform Sprite Size=500
    UsePositionOffset=false
    Write Material Random=true
   )
   &self:PartitionSimple() {module=PartitionSimple}
   &DynamicMaterialParameters(
    Index 0 Param 1=1
    Index 0 Param 2=1
    Index 0 Param 3=1
    Module.Param0WriteEnabled=true
    Module.Param1WriteEnabled=true
    Module.Param2WriteEnabled=true
    Module.Param3WriteEnabled=false
   )
  ]
  particleUpdateScript [
   &ParticleState()
   /Game/Effects/Niagara/Modules/Curves/FN_HermiteCubicSpline_v2/FN_HermiteCubicSpline(
    Position Blend:NiagaraFloat=$Particles.Partition.NormalizedID
    Position End:Vector3f=&Add_Vector(
     A=&Multiply_VectorByFloat(
      Vector=&RandomRangeVector(Minimum=<0.25 -1 -1>, Maximum=<1 1 1>, Random Seed=$Particles.RandomSeedByRibbon, Fixed Random Seed=true)
      Float=&Multiply_Float(A=150, B=&Lerp_Float(A=1, B=0.25, Alpha=$Particles.NormalizedAge))
     )
     B=<50 0 0>
    )
    Position Start:Vector3f=<25 0 0>
    Write Particles Position=true
   )
   &SolveForcesAndVelocity(Clamp Velocity=false, Limit Acceleration=false)
   /SetVariables(Particles.RibbonWidth=&FloatFromCurve(
     FloatCurve=DataInterface() {propertyValues {Curve {keys [{arriveTangent=1.7254, interpMode=RCIM_Cubic, leaveTangent=1.7254, tangentMode=RCTM_User, time=0, value=0.75}, <at 0.4 1 RCIM_Cubic>]}}}
     CurveIndex=$Particles.Partition.NormalizedID
     Scale Curve=&FloatFromCurve(
      FloatCurve=DataInterface() {propertyValues {Curve {keys [
          {arriveTangent=8.6669, interpMode=RCIM_Cubic, leaveTangent=8.6668, tangentMode=RCTM_User, time=0, value=0}
          {arriveTangent=-0.0702, interpMode=RCIM_Cubic, leaveTangent=-0.0702, tangentMode=RCTM_User, time=0.4, value=1}
          {arriveTangent=-0.8485, interpMode=RCIM_Cubic, leaveTangent=-0.8485, tangentMode=RCTM_User, time=1, value=0.25}
         ]}}}
      Scale Curve=/Niagara/DynamicInputs/Math/LerpMultipleFloats(
       Float Count:ENiagara_MultipleLerpCount=::NewEnumerator0
       1:NiagaraFloat=&RandomRangeFloat(Minimum=30, Maximum=75)
       2:NiagaraFloat=&RandomRangeFloat(Minimum=30, Maximum=75)
       3:NiagaraFloat=&RandomRangeFloat(Minimum=50, Maximum=150)
       4:NiagaraFloat=?
       5:NiagaraFloat=?
       Alpha:NiagaraFloat=$System.NormalizedDistance
      )
     )
    )) {isSetParameters=true, module=SetVariables_#7}
   &ScaleColor(
    Scale Alpha=&FloatFromCurve(
     FloatCurve=DataInterface() {propertyValues {Curve {keys [
         {arriveTangent=4.6804, interpMode=RCIM_Cubic, leaveTangent=4.6804, tangentMode=RCTM_User, time=-0.0001, value=0}
         {arriveTangent=-0.0174, interpMode=RCIM_Cubic, leaveTangent=-0.0174, tangentMode=RCTM_User, time=0.2, value=1}
         {arriveTangent=-0.0361, interpMode=RCIM_Cubic, leaveTangent=-0.0361, tangentMode=RCTM_User, time=0.6, value=1}
         {arriveTangent=0.0229, interpMode=RCIM_Cubic, leaveTangent=0.0229, tangentMode=RCTM_User, time=1, value=0}
        ]}}}
     Scale Curve=&FloatFromCurve(
      FloatCurve=DataInterface() {propertyValues {Curve {keys [
          {arriveTangent=-0.2239, interpMode=RCIM_Cubic, leaveTangent=-0.2239, tangentMode=RCTM_User, time=-0.0001, value=0}
          <at 0.5 1 RCIM_Cubic>
          {arriveTangent=-0.0174, interpMode=RCIM_Cubic, leaveTangent=-0.0174, tangentMode=RCTM_User, time=1, value=0}
         ]}}}
      CurveIndex=$Particles.Partition.NormalizedID
      Scale Curve=&FloatFromCurve(
       FloatCurve=DataInterface() {propertyValues {Curve {keys [
           {interpMode=RCIM_Cubic, tangentMode=RCTM_User, time=0, value=0.25}
           {interpMode=RCIM_Cubic, tangentMode=RCTM_User, time=0.25, value=1}
           {arriveTangent=-4.9185, interpMode=RCIM_Cubic, leaveTangent=-4.9185, tangentMode=RCTM_User, time=0.75, value=0}
          ]}}}
       CurveIndex=$System.NormalizedDistance
      )
     )
    )
    Scale RGB=&VectorFromCurve(VectorCurve=DataInterface() {propertyValues {XCurve {keys [<at 0 1>,<at 1 0>]}, YCurve {keys [<at 0 1>,<at 1 0>]}, ZCurve {keys [<at 0 1>,<at 1 1.25>]}}})
    ScaleA=true
    ScaleRGB=true
   )
  ]
  rendererClasses [&NiagaraRibbonRendererProperties]
  renderers [{rendererClass=&NiagaraRibbonRendererProperties}]
 }
 Reachers {
  emitterSpawnScript [/SetVariables(
    Emitter.EmitterRandom:NiagaraInt32=&RandomInt(Randomness Mode:ENiagaraRandomnessMode=::NewEnumerator0, Random Seed:NiagaraInt32=0, Fixed Random Seed:NiagaraBool=false)
    Emitter.ParticlesPerRibbon=5
   ) {isSetParameters=true, module=SetVariables_#5}]
  emitterUpdateScript [
   &EmitterState(Life Cycle Mode=::NewEnumerator1, Loop Duration=0.1, UseLoopDelay=false)
   &SpawnBurst_Instantaneous(
    Spawn Count=$Emitter.ParticlesPerRibbon
    Spawn Probability=&MakeCustomFloatFromBool(Bool=&SetBoolByFloatComparison(A=$System.DistanceToPlayer, B=1005), True Float=0, False Float=1)
    Use Spawn Probability=true
   )
   &SpawnBurst_Instantaneous(
    Spawn Count=&MakeCustomIntFromBool(Bool=$System.bIsBehind, True INT=0, False INT=$Emitter.ParticlesPerRibbon)
    Spawn Probability=&MakeCustomFloatFromBool(Bool=&SetBoolByFloatComparison(A=$System.DistanceToPlayer, B=1005), True Float=0, False Float=1)
    Use Spawn Probability=true
   ) {enabled=false, module=SpawnBurst_Instantaneous001}
  ]
  particleSpawnScript [
   /SetVariables(Particles.RandomSeedByRibbon:NiagaraInt32=&Add_Integer(A=$Emitter.LoopCount, B=$Emitter.EmitterRandom)) {isSetParameters=true, module=SetVariables_#6}
   /SetVariables(Particles.StoneHeldAtSpawn:NiagaraBool=$User.bIsStoneHeld) {isSetParameters=true}
   &InitializeParticle(
    AdjustAlpha=true
    AdjustHue=true
    AdjustSaturation=true
    AdjustValue=false
    Alpha Scale Range:Vector2f=<0.15 0.3>
    Color=<rgba 0.44 0 1 1>
    Color Mode=::NewEnumerator3
    Hue Shift Range:Vector2f=<-0.35 0.35>
    Lifetime=&Lerp_Float(
     A=&RandomRangeFloat(Minimum=0.1, Maximum=0.5, Randomness Mode=::NewEnumerator1, Random Seed=$Particles.RandomSeedByRibbon, Fixed Random Seed=true)
     B=&RandomRangeFloat(Minimum=1, Maximum=2, Randomness Mode=::NewEnumerator1, Random Seed=$Particles.RandomSeedByRibbon, Fixed Random Seed=true)
     Alpha=$System.NormalizedDistance
    )
    Saturation Range:Vector2f=<0.3 0.5>
    Sprite Size Mode=::NewEnumerator3
    Uniform Sprite Size=500
    UsePositionOffset=false
    Write Material Random=true
   )
   &self:PartitionSimple() {module=PartitionSimple}
   &DynamicMaterialParameters(
    Index 0 Param 1=&RandomRangeFloat(Minimum=0.5, Maximum=2)
    Index 0 Param 2=1
    Index 0 Param 3=1
    Module.Param0WriteEnabled=true
    Module.Param1WriteEnabled=true
    Module.Param2WriteEnabled=true
    Module.Param3WriteEnabled=false
   )
  ]
  particleUpdateScript [
   &ParticleState()
   /SetVariables(Particles.SmoothPlayerLocation:Vector3f=&self:SmoothInterpolateParticleVector(Initial Vector:Vector3f=$User.PlayerLocation, Vector to Interpolate:Vector3f=$User.PlayerLocation, Rate Of Change:NiagaraFloat=15)) {isSetParameters=true}
   &FN_HermiteCubicSpline(
    Position Blend:NiagaraFloat=$Particles.Partition.NormalizedID
    Position End:Vector3f=&Lerp_Vector(
     A=$Particles.SmoothPlayerLocation
     B=Linked() {linkedVariable {name=Engine.Owner.Position, type=NiagaraPosition}}
     Alpha=&Multiply_Float(
      A=$Particles.NormalizedAge
      B=&RandomRangeFloat(Minimum=0.4, Maximum=0.7, Randomness Mode=::NewEnumerator1, Random Seed=$Particles.RandomSeedByRibbon, Fixed Random Seed=true)
     )
    )
    Position Start:Vector3f=&Add_Vector(
     A=<0 0 -20>
     B=&Add_Vector(
      A=Linked() {linkedVariable {name=Engine.Owner.Position, type=NiagaraPosition}}
      B=&RandomVector(Vector Scale=25, Randomness Mode=::NewEnumerator1, Random Seed=$Particles.RandomSeedByRibbon, Fixed Random Seed=true)
     )
    )
    Write Particles Position=true
   )
   &ScaleColor(
    Scale Alpha=&FloatFromCurve(
     FloatCurve=DataInterface() {propertyValues {Curve {keys [
         {arriveTangent=4.6804, interpMode=RCIM_Cubic, leaveTangent=4.6804, tangentMode=RCTM_User, time=-0.0001, value=0}
         {arriveTangent=-0.0174, interpMode=RCIM_Cubic, leaveTangent=-0.0174, tangentMode=RCTM_User, time=0.2, value=1}
         {arriveTangent=-0.0361, interpMode=RCIM_Cubic, leaveTangent=-0.0361, tangentMode=RCTM_User, time=0.6, value=1}
         {arriveTangent=0.0229, interpMode=RCIM_Cubic, leaveTangent=0.0229, tangentMode=RCTM_User, time=1, value=0}
        ]}}}
     Scale Curve=&FloatFromCurve(
      FloatCurve=DataInterface() {propertyValues {Curve {keys [
          {arriveTangent=-0.2239, interpMode=RCIM_Cubic, leaveTangent=-0.2239, tangentMode=RCTM_User, time=-0.0001, value=0}
          {arriveTangent=-0.0174, interpMode=RCIM_Cubic, leaveTangent=-0.0174, tangentMode=RCTM_User, time=0.3, value=1}
         ]}}}
      CurveIndex=$Particles.Partition.NormalizedID
      Scale Curve=&FloatFromCurve(
       FloatCurve=DataInterface() {propertyValues {Curve {keys [
           {interpMode=RCIM_Cubic, tangentMode=RCTM_User, time=0.1, value=0}
           {interpMode=RCIM_Cubic, tangentMode=RCTM_User, time=0.25, value=1}
           {arriveTangent=-4.9185, interpMode=RCIM_Cubic, leaveTangent=-4.9185, tangentMode=RCTM_User, time=1, value=0}
          ]}}}
       CurveIndex=$System.NormalizedDistance
       Scale Curve=&Multiply_Float(
        A=&MakeCustomFloatFromBool(Bool=$System.bIsBehind, True Float=0, False Float=1)
        B=&ClampFloat(Float=$System.CompareSides)
       )
      )
     )
    )
    Scale RGB=&VectorFromCurve(VectorCurve=DataInterface() {propertyValues {XCurve {keys [<at 0 1>,<at 1 0>]}, YCurve {keys [<at 0 1>,<at 1 0>]}, ZCurve {keys [<at 0 1>,<at 1 1.25>]}}})
    ScaleA=true
    ScaleRGB=true
   )
   /SetVariables(Particles.RibbonWidth=&FloatFromCurve(
     FloatCurve=DataInterface() {propertyValues {Curve {keys [
         {arriveTangent=-0.0607, interpMode=RCIM_Cubic, leaveTangent=-0.0607, tangentMode=RCTM_User, time=0, value=1.197}
         {arriveTangent=-1.4893, interpMode=RCIM_Cubic, leaveTangent=-1.4893, tangentMode=RCTM_User, time=0.4, value=1}
         {arriveTangent=-1.3134, interpMode=RCIM_Cubic, leaveTangent=-1.3134, tangentMode=RCTM_User, time=1, value=0.2654}
        ]}}}
     CurveIndex=$Particles.Partition.NormalizedID
     Scale Curve=&FloatFromCurve(
      FloatCurve=DataInterface() {propertyValues {Curve {keys [
          {arriveTangent=0.0151, interpMode=RCIM_Cubic, leaveTangent=0.0151, tangentMode=RCTM_User, time=0, value=0}
          {arriveTangent=-0.0702, interpMode=RCIM_Cubic, leaveTangent=-0.0702, tangentMode=RCTM_User, time=0.4, value=1}
          {arriveTangent=-0.8485, interpMode=RCIM_Cubic, leaveTangent=-0.8485, tangentMode=RCTM_User, time=1, value=0.25}
         ]}}}
      Scale Curve=&LerpMultipleFloats(
       Float Count:ENiagara_MultipleLerpCount=::NewEnumerator0
       1:NiagaraFloat=&RandomRangeFloat(Minimum=20, Maximum=50, Randomness Mode=::NewEnumerator1, Random Seed=$Particles.RandomSeedByRibbon, Fixed Random Seed=true)
       2:NiagaraFloat=&RandomRangeFloat(Minimum=20, Maximum=50, Randomness Mode=::NewEnumerator1, Random Seed=$Particles.RandomSeedByRibbon, Fixed Random Seed=true)
       3:NiagaraFloat=&RandomRangeFloat(Minimum=40, Maximum=100, Randomness Mode=::NewEnumerator1, Random Seed=$Particles.RandomSeedByRibbon, Fixed Random Seed=true)
       4:NiagaraFloat=?
       5:NiagaraFloat=?
       Alpha:NiagaraFloat=$System.NormalizedDistance
      )
     )
    )) {isSetParameters=true, module=SetVariables_#7}
   &Color(Color=&ColorFromCurve(
     ColorCurve=DataInterface() {propertyValues {
       AlphaCurve {keys [<at 0 1 RCIM_Cubic>]}
       BlueCurve {keys [<at 0 0 RCIM_Cubic>,<at 0.4908 0>,<at 1 1 RCIM_Cubic>]}
       GreenCurve {keys [<at 0 0 RCIM_Cubic>,<at 0.4908 1>,<at 1 0 RCIM_Cubic>]}
       RedCurve {keys [<at 0 1 RCIM_Cubic>,<at 0.4908 0>,<at 1 0 RCIM_Cubic>]}
      }}
     CurveIndex=&MakeCustomFloatFromBool(Bool=$User.bIsStoneHeld, True Float=1, False Float=0)
    )) {enabled=false}
  ]
  rendererClasses [&NiagaraRibbonRendererProperties, &NiagaraSpriteRendererProperties]
  renderers [
   {rendererClass=&NiagaraRibbonRendererProperties}
   {rendererClass=&NiagaraSpriteRendererProperties, rendererIndex=1}
  ]
 }
 Streaks {
  emitterUpdateScript [
   &EmitterState()
   &SpawnRate(SpawnRate=75, Use Spawn Probability=false)
  ]
  particleSpawnScript [
   &InitializeParticle(
    AdjustAlpha=true
    AdjustHue=true
    AdjustSaturation=true
    AdjustValue=false
    Alpha Scale Range:Vector2f=<0.25 1>
    Color=<rgba 0.44 0 1 1>
    Color Mode=::NewEnumerator3
    Hue Shift Range:Vector2f=<-0.35 0.35>
    Lifetime Max=0.25
    Lifetime Min=0.1
    Lifetime Mode=::NewEnumerator1
    Saturation Range:Vector2f=<0.3 0.5>
    Sprite Size Max=<2 40>
    Sprite Size Min=<1 20>
    Sprite Size Mode=::NewEnumerator2
    UsePositionOffset=false
    Write Material Random=true
   )
   &AddVelocity(Velocity=<100 0 0>) {enabled=false}
   &SphereLocation(
    MaskSpawnGroup=false
    Offset=<104.3615 0 0>
    Override Seed=false
    Sphere Radius=80
    Surface Only=false
    UseNonUniformScale=false
    UseOffset=true
    UseSphereOrientation=false
   )
   &DynamicMaterialParameters(
    Index 0 Param 1=&RandomRangeFloat(Minimum=0.5, Maximum=2)
    Index 0 Param 2=20
    Module.Param0WriteEnabled=true
    Module.Param1WriteEnabled=true
    Module.Param2WriteEnabled=false
    Module.Param3WriteEnabled=false
   )
  ]
  particleUpdateScript [
   &ParticleState()
   &ScaleColor(
    Scale Alpha=&FloatFromCurve(
     FloatCurve=DataInterface() {propertyValues {Curve {keys [<at 0 0 RCIM_Cubic>,<at 1 1 RCIM_Cubic>]}}}
     Scale Curve=&ClampFloat(Float=&Add_Float(A=1.25, B=&Multiply_Float(A=$System.DistanceToPlayer, B=-0.001)))
    )
    ScaleA=true
    ScaleRGB=false
   )
   &ScaleSpriteSize(Scale Factor=&Vector2DFromCurve(
     Vector2Curve=DataInterface() {propertyValues {XCurve {keys [<at 0.5 1>,<at 1 0>]}, YCurve {keys [<at 0 1>,<at 1 1>]}}}
     ScaleCurve=&Vector2DFromFloat(Value=&RandomRangeFloat(Minimum=1, Maximum=2))
    ))
   /SetVariables(
    Particles.Position=&Lerp_Vector(
     A=Linked() {linkedVariable {name=Particles.Position, type=NiagaraPosition}}
     B=Linked() {linkedVariable {name=Engine.Owner.Position, type=NiagaraPosition}}
     Alpha=&Multiply_Float(
      A=&FloatFromCurve(FloatCurve=DataInterface() {propertyValues {Curve {keys [
           {interpMode=RCIM_Cubic, tangentMode=RCTM_User, time=0, value=0}
           {arriveTangent=2.2555, interpMode=RCIM_Cubic, leaveTangent=2.2555, tangentMode=RCTM_User, time=1, value=1}
          ]}}})
      B=&RandomRangeFloat(Minimum=0.3, Maximum=0.5)
     )
    )
    Particles.SpriteAlignment=&Subtract_Vector(
     A=Linked() {linkedVariable {name=Particles.Position, type=NiagaraPosition}}
     B=Linked() {linkedVariable {name=Engine.Owner.Position, type=NiagaraPosition}}
    )
   ) {isSetParameters=true}
  ]
  rendererClasses [&NiagaraSpriteRendererProperties]
  renderers [{rendererClass=&NiagaraSpriteRendererProperties}]
 }
}
summary {
 systemName=NS_TowerObelisk_Suction
 emitters [
  {emitterName=Corona, rendererClasses [&NiagaraSpriteRendererProperties, &NiagaraLightRendererProperties]}
  {
   bEnabled=false
   emitterName=Empty
   rendererClasses [&NiagaraSpriteRendererProperties, &NiagaraMeshRendererProperties]
  }
  {emitterName=Streaks, rendererClasses [&NiagaraSpriteRendererProperties]}
  {bEnabled=false, emitterName=Empty002, rendererClasses [&NiagaraSpriteRendererProperties]}
  {bEnabled=false, emitterName=IdleTentacles, rendererClasses [&NiagaraRibbonRendererProperties]}
  {bEnabled=false, emitterName=Empty001, rendererClasses [&NiagaraSpriteRendererProperties]}
  {emitterName=Reachers, rendererClasses [&NiagaraRibbonRendererProperties, &NiagaraSpriteRendererProperties]}
 ]
 userVariables [
  User.bIsStoneHeld {type=NiagaraBool}
  User.PlayerLocation {type=Vector3f, description="", defaultValue {struct=Vector3f, value=<518.0515 20 75>}}
 ]
}
